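# Niagara charge effect for the PintPiano emote, combining Boosters and Thruster exhausts with a front cone and flares.
# Boosters and Thruster use meshes and sprites at fixed offsets, pulsed on and off by loop-age curves over ten seconds, with waveform oscillation on the thrusters.
# FrontCone and Front_Flares start after a delay, with flare spawn rate ramping up.
# Time-of-day scalars adjust brightness; the Flare emitter has no renderers.
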NiagaraSystem /BRCosmetics/Animation/Game/MainPlayer/Emotes/PintPiano/FX/NS_Emote_PintPiano_Charge
dependencies {
 usedDataInterfaces [
  &NiagaraDataInterfaceCurve
  &NiagaraDataInterfaceColorCurve
  &NiagaraDataInterfaceMeshRendererInfo
  &NiagaraDataInterfaceArrayFloat3
 ]
 usedDynamicInputs [&SimulationPosition]
 usedModules [&SystemState]
}
emitters {
 Boosters {
  emitterUpdateScript [
   &EmitterState(Inactive Response=::NewEnumerator1, Life Cycle Mode=::NewEnumerator1, Loop Behavior=::NewEnumerator1, Loop Duration=10, UseLoopDelay=false)
   &SpawnBurst_Instantaneous(Spawn Group=1, Use Spawn Probability=false, UseLoopCountLimit=false) {module=SpawnBurst_Instantaneous001}
   &SpawnBurst_Instantaneous(Spawn Count=4, Use Spawn Probability=false, UseLoopCountLimit=false)
   /Niagara/Noise/Waveform(
    EnableExponent1=false
    Global Amplitude Scale=0
    GlobalClampMax=false
    GlobalClampMin=false
    UseInput1=false
    UsePhase1=false
    "Waveform [1]":ENiagara_Waveforms=::NewEnumerator12
    "[1] Frequency"=7
   )
   /SetVariables(Emitter.Flicker:NiagaraFloat=&Add_Float(A=&Multiply_Float(A=$Emitter.Waveform.Waveform, B=0.5), B=0.5)) {isSetParameters=true}
   /SetVariables(Emitter.LightVisibility:NiagaraBool=/Niagara/DynamicInputs/Helpers/MaskBool(
     Mask Source:NiagaraInt32=1
     Mask Parameter:NiagaraInt32=&MakeCustomIntFromBool(Bool=$User.bIsFrontEndPreview, True INT=1, False INT=0)
     Mask:NiagaraInt32=1
     Masked Bool:NiagaraBool=false
     Passthrough Bool:NiagaraBool=true
    )) {isSetParameters=true}
  ]
  particleSpawnScript [
   /SetVariables(Particles.VisibilityTag=&SpawnGroup()) {enabled=false, isSetParameters=true}
   &PartitionParticles(Number Of Partitions=4)
   &InitializeParticle(
    Color=<rgba 8 4.65 0 1>
    Lifetime=9.7
    Mesh Scale=&Multiply_VectorByFloat(Vector=<0.875 0.875 1.5>, Float=0.05)
    Mesh Scale Mode=::NewEnumerator4
    Sprite Rotation Mode=::NewEnumerator0
    Sprite Size=<90 175>
    Sprite Size Mode=::NewEnumerator4
    UsePositionOffset=false
    Write Material Random=true
   )
   &InitialMeshOrientation(Mesh Axis To Orient=<0 0 1>, Mesh Orientation Mode=::NewEnumerator2, Orientation Vector=<0 1 0>, Use Rotation Vector=false)
   &SystemLocation(
    MaskSpawnGroup=false
    Offset=&MergeHelper_VectorSwitch(0=<20 17 15>, 1=<-25 17 15>, 2=<20 17 -15>, 3=<-25 17 -15>, Index=$Particles.PartitionParticles.Partition, Num=::NewEnumerator2)
   )
  ]
  particleUpdateScript [
   &ParticleState()
   &ScaleMeshSize(Scale Factor=&Multiply_VectorByFloat(
     Vector=<2 2 2>
     Float=&Multiply_Float(
      A=&FloatFromCurve(
       FloatCurve=DataInterface() {propertyValues {Curve {
          defaultValue=1
          keys [
           {interpMode=RCIM_Cubic, tangentMode=RCTM_User, time=0, value=0}
           {interpMode=RCIM_Cubic, tangentMode=RCTM_User, time=0.1089, value=1.7094}
           <at 0.2301 1 RCIM_Cubic>
          ]
         }}}
       CurveIndex=$Emitter.Age
      )
      B=&FloatFromCurve(
       FloatCurve=DataInterface() {propertyValues {Curve {keys [<at 0 1 RCIM_Constant>,<at 0.3858 0 RCIM_Constant>,<at 0.5946 1 RCIM_Constant>,<at 0.9361 0 RCIM_Constant>]}}}
       CurveIndex=$Emitter.NormalizedLoopAge
      )
     )
    ))
   &SolveForcesAndVelocity(Clamp Velocity=false, Limit Acceleration=false)
   &ScaleColor(
    Scale Alpha=&Multiply_Float(
     A=$Emitter.Flicker
     B=&Multiply_Float(
      A=&FNTimeOfDayScalar(Night Scale=2, Morning Scale=3, Day Scale=4, Evening Scale=3)
      B=&FloatFromCurve(
       FloatCurve=DataInterface() {propertyValues {Curve {keys [<at 0 1 RCIM_Constant>,<at 0.3858 0 RCIM_Constant>,<at 0.5946 1 RCIM_Constant>,<at 0.9361 0 RCIM_Constant>]}}}
       CurveIndex=$Emitter.NormalizedLoopAge
      )
     )
    )
    ScaleA=true
    ScaleRGB=true
   )
   /SetVariables(Particles.FlameScale:Vector3f=&Multiply_VectorByFloat(
     Vector=&MakeVector(
      X=2
      Y=2
      Z=&Multiply_Float(
       A=1
       B=&Lerp_Float(
        A=0
        B=2
        Alpha=&FloatFromCurve(
         FloatCurve=DataInterface() {propertyValues {Curve {keys [<at 0 1 RCIM_Constant>,<at 0.3858 0 RCIM_Constant>,<at 0.5946 1 RCIM_Constant>,<at 0.9361 0 RCIM_Constant>]}}}
         CurveIndex=$Emitter.NormalizedLoopAge
        )
       )
      )
     )
     Float=&FloatFromCurve(
      FloatCurve=DataInterface() {propertyValues {Curve {
         defaultValue=1
         keys [
          {interpMode=RCIM_Cubic, tangentMode=RCTM_User, time=0, value=0}
          {interpMode=RCIM_Cubic, tangentMode=RCTM_User, time=0.1916, value=1.3594}
          <at 0.3265 1 RCIM_Cubic>
         ]
        }}}
      CurveIndex=$Emitter.Age
      Scale Curve=0.05
     )
    )) {isSetParameters=true}
   /SetVariables(
    Particles.Color=<rgba 3 0.8675 0.2213 1>
    Particles.SpritePosition:NiagaraPosition=&AddVectorToPosition(Position=$Particles.Position, Vector=<0 1 0>)
    Particles.SpriteSize=&Multiply_Vector2D_ByFloat(
     Float=&Multiply_Float(
      A=12
      B=&FloatFromCurve(
       FloatCurve=DataInterface() {propertyValues {Curve {keys [
           {interpMode=RCIM_Cubic, leaveTangentWeight=0.1146, tangentMode=RCTM_User, time=0, value=0}
           {arriveTangentWeight=0.3184, interpMode=RCIM_Cubic, leaveTangentWeight=0.0389, tangentMode=RCTM_User, time=0.3631, value=0}
           {arriveTangentWeight=0.014, interpMode=RCIM_Cubic, leaveTangentWeight=0.0804, tangentMode=RCTM_User, time=0.3922, value=0}
           {arriveTangentWeight=0.2032, interpMode=RCIM_Cubic, leaveTangentWeight=0.1397, tangentMode=RCTM_User, time=0.5608, value=0}
           {arriveTangentWeight=0.1747, interpMode=RCIM_Cubic, leaveTangentWeight=0.2057, tangentMode=RCTM_User, time=0.5882, value=1}
           {arriveTangentWeight=0.3203, interpMode=RCIM_Cubic, leaveTangentWeight=0.0286, tangentMode=RCTM_User, time=0.8721, value=1.0134}
           {arriveTangentWeight=0.0085, interpMode=RCIM_Cubic, tangentMode=RCTM_User, time=0.891, value=0}
          ]}}}
       CurveIndex=$Emitter.NormalizedLoopAge
      )
     )
     Vector2D=<1 1>
    )
   ) {isSetParameters=true}
  ]
  rendererClasses [&NiagaraMeshRendererProperties, &NiagaraSpriteRendererProperties]
  renderers [
   {rendererClass=&NiagaraMeshRendererProperties, rendererIndex=0}
   {rendererClass=&NiagaraMeshRendererProperties, rendererIndex=1}
   {rendererClass=&NiagaraSpriteRendererProperties, rendererIndex=2}
  ]
 }
 Flare {rendererClasses [], renderers []}
 FrontCone {
  emitterUpdateScript [
   &EmitterState(
    Inactive Response=::NewEnumerator1
    Life Cycle Mode=::NewEnumerator1
    Loop Behavior=::NewEnumerator1
    Loop Delay=5
    Loop Duration=4.4
    UseLoopDelay=true
   )
   &SpawnBurst_Instantaneous(Use Spawn Probability=false, UseLoopCountLimit=false)
  ]
  particleSpawnScript [&InitializeParticle(
    Lifetime=4.4
    Mesh Scale=&Multiply_Vector(A=<0.2 0.15 0.2>, B=<0.5 1 0.75>)
    Mesh Scale Mode=::NewEnumerator4
    Position Offset=<33 5 0>
    UsePositionOffset=true
    Write Material Random=true
   )]
  particleUpdateScript [
   &ParticleState()
   &ScaleMeshSize(Scale Factor=<2 2 2>)
   &DynamicMaterialParameters(
    Index 0 Param 2=1
    Index 1 Param 3=&Lerp_Float(
     A=1.5
     B=0.7
     Alpha=&FloatFromCurve(FloatCurve=DataInterface() {propertyValues {Curve {keys [
          {arriveTangent=2.0732, interpMode=RCIM_Cubic, leaveTangent=2.0732, tangentMode=RCTM_User, time=0, value=0}
          {arriveTangent=-0.2438, interpMode=RCIM_Cubic, leaveTangent=-0.2438, tangentMode=RCTM_User, time=1, value=1}
         ]}}})
    )
    Module.Index0Float0Write=true
    Module.Index0Float1Write=true
    Module.Index0Float2Write=true
    Module.Index0Float3Write=true
    Module.Index1Float0Write=true
    Module.Index1Float1Write=true
    Module.Index1Float2Write=true
    Module.Index1Float3Write=true
    Write Parameter Index 1=::NewEnumerator1
   )
   &Color(Color=&Multiply_LinearColorByFloat(
     Linear Color=&ColorFromCurve(ColorCurve=DataInterface() {propertyValues {
        AlphaCurve {keys [<at 0 1 RCIM_Cubic>]}
        BlueCurve {keys [<at 0.4077 1 RCIM_Cubic>,<at 1 2 RCIM_Cubic>]}
        GreenCurve {keys [<at 0.4077 1 RCIM_Cubic>,<at 1 2 RCIM_Cubic>]}
        RedCurve {keys [<at 0.4077 1 RCIM_Cubic>,<at 1 2 RCIM_Cubic>]}
       }})
     Float=&FloatFromCurve(
      FloatCurve=DataInterface() {propertyValues {Curve {keys [{interpMode=RCIM_Cubic, tangentMode=RCTM_User, time=0, value=0}, {interpMode=RCIM_Cubic, tangentMode=RCTM_User, time=0.2, value=1}]}}}
      CurveIndex=$Engine.System.Age
      Scale Curve=&FNTimeOfDayScalar(Night Scale=0.8, Morning Scale=0.8, Evening Scale=0.8)
     )
    ))
   /SetVariables(Particles.ColorHighlight:LinearColor=&Multiply_LinearColorByFloat(
     Linear Color=$Particles.Color
     Float=&FNTimeOfDayScalar(Night Scale=2, Morning Scale=2, Day Scale=3, Evening Scale=2)
    )) {isSetParameters=true}
  ]
  rendererClasses [&NiagaraMeshRendererProperties]
  renderers [
   {rendererClass=&NiagaraMeshRendererProperties, rendererIndex=0}
   {rendererClass=&NiagaraMeshRendererProperties, rendererIndex=1}
  ]
 }
 Front_Flares {
  emitterUpdateScript [
   &EmitterState(Life Cycle Mode=::NewEnumerator1, Loop Behavior=::NewEnumerator1, Loop Delay=5, Loop Duration=4.4, UseLoopDelay=true)
   &SpawnRate(
    SpawnRate=&FloatFromCurve(
     FloatCurve=DataInterface() {propertyValues {Curve {
        defaultValue=1
        keys [
         {arriveTangent=0.0667, interpMode=RCIM_Cubic, leaveTangent=0.0667, tangentMode=RCTM_User, time=0, value=0.25}
         {arriveTangent=1.9282, tangentMode=RCTM_User, time=1, value=1}
        ]
       }}}
     CurveIndex=$Emitter.NormalizedLoopAge
     Scale Curve=10
    )
    Use Spawn Probability=false
   )
  ]
  particleSpawnScript [
   &InitializeParticle(
    Color Maximum=<rgba 0.8738 0.5896 4 1>
    Color Minimum=<rgba 0.851 2.9503 4 1>
    Color Mode=::NewEnumerator2
    Lifetime Max=0.6
    Lifetime Min=0.4
    Lifetime Mode=::NewEnumerator1
    Mesh Renderer Array Visibility Mode=::NewEnumerator1
    Mesh Renderer Info=DataInterface() {propertyValues {MeshRenderer=&self:Front_Flares_0.NiagaraMeshRendererProperties_0}}
    Mesh Renderer Mesh Index:NiagaraInt32=&MakeIntFromFloat(
     Float=&FloatFromCurve(
      FloatCurve=DataInterface() {propertyValues {Curve {defaultValue=1, keys [<at 0 0 RCIM_Cubic>, {arriveTangent=2.8878, tangentMode=RCTM_User, time=1, value=1}]}}}
      CurveIndex=&RandomRangeFloat(Minimum=0, Maximum=1)
     )
     Conversion Mode=::NewEnumerator1
    )
    Mesh Scale=&Lerp_Vector(
     A=<0.8 0.8 0.8>
     B=<1 1 1>
     Alpha=&FloatFromCurve(
      FloatCurve=DataInterface() {propertyValues {Curve {
         defaultValue=1
         keys [
          {arriveTangent=0.4663, interpMode=RCIM_Cubic, leaveTangent=0.4663, tangentMode=RCTM_User, time=0, value=0}
          {arriveTangent=2.2578, tangentMode=RCTM_User, time=1, value=1}
         ]
        }}}
      CurveIndex=&RandomRangeFloat(Minimum=0, Maximum=1)
     )
    )
    Mesh Scale Mode=::NewEnumerator4
    Position Offset=<33 5 0>
    Sprite Rotation Mode=::NewEnumerator2
    Sprite Size Mode=::NewEnumerator3
    Uniform Sprite Size=&FloatFromCurve(
     FloatCurve=DataInterface() {propertyValues {Curve {
        defaultValue=1
        keys [
         {arriveTangent=368.9517, interpMode=RCIM_Cubic, leaveTangent=368.9517, tangentMode=RCTM_User, time=0, value=50}
         {arriveTangent=-20.268, interpMode=RCIM_Cubic, leaveTangent=-20.268, tangentMode=RCTM_User, time=1, value=200}
        ]
       }}}
     CurveIndex=&RandomRangeFloat(Minimum=0, Maximum=1)
     Scale Curve=0.25
    )
    UsePositionOffset=true
    Write Material Random=true
   )
   &ShapeLocation(
    Cylinder Height=5
    Cylinder Radius=5
    MaskSpawnGroup=false
    Offset=&SelectVectorFromArray(
     Array Sampling Mode=::NewEnumerator0
     Direct Array Index=?
     Vector Selection Array=DataInterface() {propertyValues {FloatData [<-10 -10 0>, <0 0 0>]}}
     Random Seed=0
    )
    Offset Mode=::NewEnumerator0
    Override Seed=false
    Shape Primitive=::NewEnumerator1
    Surface Only=false
    UseOwnerScale=true
   )
   &CameraOffset(Camera Offset Amount=10)
  ]
  particleUpdateScript [
   &ParticleState()
   &ScaleSpriteSize(Uniform Curve Sprite Scale=DataInterface() {propertyValues {Curve {
       defaultValue=0
       keys [
        {arriveTangent=4.1785, interpMode=RCIM_Cubic, leaveTangent=4.1785, tangentMode=RCTM_User, time=0, value=0}
        <at 0.1 1 RCIM_Cubic>
        <at 0.5 0 RCIM_Cubic>
       ]
      }}})
   &ScaleColor(
    Scale Alpha=&FloatFromCurve(
     FloatCurve=DataInterface() {propertyValues {Curve {
        defaultValue=1
        keys [
         {arriveTangent=9.0725, interpMode=RCIM_Cubic, leaveTangent=9.0725, tangentMode=RCTM_User, time=0, value=0}
         <at 0.1408 0.9847 RCIM_Cubic>
         {arriveTangent=-0.0405, interpMode=RCIM_Cubic, leaveTangent=-0.0405, tangentMode=RCTM_User, time=1, value=0}
        ]
       }}}
     Scale Curve=0.8
    )
    Scale RGB=&VectorFromFloat(Value=&FloatFromCurve(
      FloatCurve=DataInterface() {propertyValues {Curve {
         defaultValue=1
         keys [
          {arriveTangent=9.0725, interpMode=RCIM_Cubic, leaveTangent=9.0725, tangentMode=RCTM_User, time=0, value=0}
          <at 0.0916 0.9847 RCIM_Cubic>
          {arriveTangent=-0.0405, interpMode=RCIM_Cubic, leaveTangent=-0.0405, tangentMode=RCTM_User, time=1, value=0}
         ]
        }}}
      Scale Curve=2
     ))
    ScaleA=true
    ScaleRGB=true
   )
   /SetVariables(Particles.Color - Flare:LinearColor=&MultiplyLinearColor(
     Linear Color=$Particles.Initial.Color
     Scale Factor=&MakeVector4FromVectorAndFloat(
      "Vector (RGB)"=&VectorFromFloat(Value=&FloatFromCurve(
        FloatCurve=DataInterface() {propertyValues {Curve {
           defaultValue=1
           keys [
            {arriveTangent=0.556, interpMode=RCIM_Cubic, leaveTangent=0.556, tangentMode=RCTM_User, time=0, value=0}
            {arriveTangent=-0.2562, interpMode=RCIM_Cubic, leaveTangent=-0.2562, tangentMode=RCTM_User, time=0.1016, value=0.8685}
            {arriveTangent=-0.0405, interpMode=RCIM_Cubic, leaveTangent=-0.0405, tangentMode=RCTM_User, time=0.4, value=0}
           ]
          }}}
        Scale Curve=&RandomRangeFloat(Minimum=1, Maximum=2)
       ))
      "Float (Alpha)"=&FNTimeOfDayScalar(Night Scale=1, Morning Scale=1.5, Day Scale=2, Evening Scale=1.5)
     )
    )) {isSetParameters=true}
  ]
  rendererClasses [&NiagaraSpriteRendererProperties]
  renderers [{rendererClass=&NiagaraSpriteRendererProperties}]
 }
 Thruster {
  emitterSpawnScript [/SetVariables() {isSetParameters=true}]
  emitterUpdateScript [
   &EmitterState(Inactive Response=::NewEnumerator1, Life Cycle Mode=::NewEnumerator1, Loop Behavior=::NewEnumerator1, Loop Duration=10, UseLoopDelay=true)
   &SpawnBurst_Instantaneous(Spawn Count=2, Use Spawn Probability=false)
   /SetVariables(Emitter.SmoothVelocity:Vector3f=/Niagara/DynamicInputs/Math/Clamp_Vector(
     Vector:Vector3f=&Lerp_Vector(
      A=$Emitter.SmoothVelocity
      B=&TransformVector(Vector=$Engine.Owner.Velocity, Source Space=::World, Destination Space=::Local)
      Alpha=0.06
     )
     Min:Vector3f=<0 0 0>
     Max:Vector3f=<1000 1000 1000>
    )) {isSetParameters=true}
   /SetVariables(Emitter.ForwardDot:NiagaraFloat=&ClampFloat(Float=&Divide_Float(
      A=&Subtract_Float(A=&MakeFloatFromVector(Vector=$Emitter.SmoothVelocity, Channel=::NewEnumerator0), B=50)
      B=500
     ))) {isSetParameters=true}
  ]
  particleSpawnScript [
   &PartitionParticles(Particles Per Partition=1, Partition Mode:ENiagara_PartitionMode_v2=::NewEnumerator1)
   &InitializeParticle(Lifetime=9.7, Mesh Scale=<0.3 0.7 0.75>, Mesh Scale Mode=::NewEnumerator4, UsePositionOffset=false, Write Material Random=true) {module=InitializeParticle001}
   &InitialMeshOrientation(Mesh Orientation Mode=::NewEnumerator2, Orientation Vector=<-1 0 0>, Use Rotation Vector=true) {module=InitialMeshOrientation001}
   &SystemLocation(
    MaskSpawnGroup=false
    Offset=&MaskVector(Mask Source=1, Mask Parameter=$Particles.PartitionParticles.Partition, Mask=0, Masked Vector=<-31 -20 -7>, Passthrough Vector=<-31 -20 7>)
   )
  ]
  particleUpdateScript [
   &ParticleState()
   /SetVariables(Particles.Oscillation:NiagaraFloat=&Add_Float(
     A=/Niagara/DynamicInputs/Helpers/Waveform(
      Waveform Count=::NewEnumerator1
      Global Phase=$Particles.Age
      "Waveform [1]"=::NewEnumerator0
      "[1] Frequency"=2.256
      "[1] Amplitude Min/Max"=<0 1>
      "[1] Scale"=0.5
      "[1] Bias"=1
      "Waveform [2] Blend"=::NewEnumerator2
      "[2] Input Position"=0
      "[2] Phase"=$Emitter.Age
      "[2] Phase Scale"=1
      "[2] Compound Frequency"=2
      "[2] Frequency"=7.2563
      "[2] Amplitude Min/Max"=<0 1>
      "[2] Scale"=0.5
      "[2] Bias"=1
      "[2] Exponent"=1
      "[2] Pulse Duration"=0.1
      "[2] Clip Triangle Min/Max"=<0 0>
      "[2] Random Seed"=0
     )
     B=0
    )) {isSetParameters=true}
   &ScaleMeshSize(Scale Factor=&Multiply_Vector(
     A=<7 1 1>
     B=&Multiply_Vector(
      A=&Lerp_Vector(A=<0.2 1 1>, B=<1 1 1>, Alpha=$Particles.Oscillation)
      B=&Lerp_Vector(
       A=<1 1 1>
       B=<3 1 1>
       Alpha=&FloatFromCurve(
        FloatCurve=DataInterface() {propertyValues {Curve {
           defaultValue=1
           keys [
            {interpMode=RCIM_Cubic, tangentMode=RCTM_User, time=0, value=1}
            {
             arriveTangent=-23.8597
             arriveTangentWeight=0.1535
             interpMode=RCIM_Cubic
             leaveTangentWeight=0.1888
             tangentMode=RCTM_Break
             time=0.1044
             value=0
            }
            {arriveTangentWeight=0.2092, interpMode=RCIM_Cubic, leaveTangentWeight=0.1283, tangentMode=RCTM_User, time=0.2084, value=-0.0029}
            {arriveTangentWeight=0.1456, interpMode=RCIM_Cubic, leaveTangentWeight=0.1893, tangentMode=RCTM_User, time=0.2285, value=0.7}
            {arriveTangentWeight=0.1402, interpMode=RCIM_Cubic, leaveTangentWeight=0.1942, tangentMode=RCTM_User, time=0.2624, value=0.7}
            {arriveTangentWeight=0.0765, interpMode=RCIM_Cubic, leaveTangentWeight=0.2641, tangentMode=RCTM_User, time=0.2851, value=0}
            {arriveTangentWeight=0.1091, interpMode=RCIM_Cubic, leaveTangentWeight=0.2518, tangentMode=RCTM_User, time=0.4352, value=0}
            <at 0.4738 1 RCIM_Cubic>
           ]
          }}}
        CurveIndex=$Emitter.NormalizedLoopAge
       )
      )
     )
    ))
   &SolveForcesAndVelocity(Clamp Velocity=false, Limit Acceleration=false)
   &DynamicMaterialParameters(
    Index 0 Param 1=&RandomRangeFloat(Minimum=0, Maximum=1)
    Index 0 Param 2=5
    Index 0 Param 3=5
    Module.Index0Float0Write=true
    Module.Index0Float1Write=true
    Module.Index0Float2Write=true
    Module.Index0Float3Write=true
    Module.Index3Float0Write=true
    Module.Index3Float1Write=true
    Module.Index3Float2Write=true
    Module.Index3Float3Write=true
    Write Parameter Index 3=::NewEnumerator1
   )
   &ScaleColor(
    Scale Alpha=&Multiply_Float(
     A=&FNTimeOfDayScalar(Night Scale=0.3, Morning Scale=0.7, Evening Scale=0.7)
     B=&FloatFromCurve(
      FloatCurve=DataInterface() {propertyValues {Curve {keys [
          <at 0 1>
          {arriveTangent=-1.8686, interpMode=RCIM_Cubic, leaveTangent=-1.8686, time=0.4835, value=0.688}
          {arriveTangentWeight=0.1044, interpMode=RCIM_Cubic, leaveTangentWeight=0.2485, tangentMode=RCTM_User, time=0.5352, value=0}
          {arriveTangentWeight=0.0409, interpMode=RCIM_Cubic, leaveTangentWeight=0.3123, tangentMode=RCTM_User, time=0.9004, value=0}
          {arriveTangent=8.7612, interpMode=RCIM_Cubic, leaveTangent=8.7612, tangentMode=RCTM_User, time=0.9363, value=0.8026}
          <at 0.995 1.0037 RCIM_Cubic>
         ]}}}
      CurveIndex=$Emitter.NormalizedLoopAge
     )
    )
    Scale RGB=&MakeVectorFromLinearColorRGB(Color=<rgba 2 0.4449 0.1456 1>)
    ScaleA=true
    ScaleRGB=true
   )
   &DynamicMaterialParameters(
    Index 0 Param 1=&RandomRangeFloat(Minimum=0, Maximum=1)
    Index 3 Param 1=3
    Index 3 Param 2=50
    Index 3 Param 3=1
    Module.Index0Float0Write=true
    Module.Index0Float1Write=true
    Module.Index0Float2Write=true
    Module.Index0Float3Write=true
    Module.Index3Float0Write=true
    Module.Index3Float1Write=true
    Module.Index3Float2Write=true
    Module.Index3Float3Write=true
    Write Parameter Index 3=::NewEnumerator1
   ) {module=DynamicMaterialParameters001}
   /NiagaraFN/Sprite/FN_SpriteAlignmentAndFacing(SpriteAlignment:Vector3f=<0 0 1>, SpriteFacing:Vector3f=<-1 0 0>)
   &FN_Sprite_Attributes(
    Alignment:Vector3f=<0 1 0>
    Camera Offset=5
    Create Alignment Attribute=true
    Create Camera Offset Attribute=true
    Create Position Attribute=true
    Create SpriteSize Attribute=true
    Position:NiagaraPosition=&AddVectorToPosition(Position=$Particles.Position, Vector=<-3 0 0>)
    SpriteSize=&Multiply_Vector2D_ByFloat(
     Float=&Multiply_Float(A=0.4, B=&Lerp_Float(A=1, B=1.2, Alpha=$Particles.Oscillation))
     Vector2D=<60 80>
    )
    Write to Module Namespace=true
   )
   &FN_Sprite_Attributes(
    Camera Offset=10
    Color=<rgba 1 0.261 0 1>
    Create Camera Offset Attribute=true
    Create Color Attribute=true
    Create Position Attribute=true
    Create SpriteSize Attribute=true
    Position:NiagaraPosition=&AddVectorToPosition(
     Position=$Particles.Position
     Vector=&MaskVector(Mask Source=1, Mask Parameter=$Particles.PartitionParticles.Partition, Mask=0, Masked Vector=<0 0 0>, Passthrough Vector=<0 1.5631 -1>)
    )
    SpriteSize=&Multiply_Vector2D_ByFloat(
     Float=&Multiply_Float(
      A=&Lerp_Float(A=1, B=1.2, Alpha=$Particles.Oscillation)
      B=&FloatFromCurve(
       FloatCurve=DataInterface() {propertyValues {Curve {keys [
           <at 0 1>
           {arriveTangent=-1.8686, interpMode=RCIM_Cubic, leaveTangent=-1.8686, time=0.4835, value=0.688}
           {arriveTangentWeight=0.1044, interpMode=RCIM_Cubic, leaveTangentWeight=0.2485, tangentMode=RCTM_User, time=0.5352, value=0}
           {arriveTangentWeight=0.0409, interpMode=RCIM_Cubic, leaveTangentWeight=0.3123, tangentMode=RCTM_User, time=0.9004, value=0}
           {arriveTangent=8.7612, interpMode=RCIM_Cubic, leaveTangent=8.7612, tangentMode=RCTM_User, time=0.9363, value=0.8026}
           <at 0.995 1.0037 RCIM_Cubic>
          ]}}}
       CurveIndex=$Emitter.NormalizedLoopAge
       Scale Curve=&FNTimeOfDayScalar(Night Scale=0.8, Morning Scale=0.8, Evening Scale=0.8)
      )
     )
     Vector2D=<100 5>
    )
    Write to Module Namespace=true
   ) {module=FN_Sprite_Attributes002}
  ]
  rendererClasses [&NiagaraMeshRendererProperties, &NiagaraSpriteRendererProperties]
  renderers [
   {rendererClass=&NiagaraMeshRendererProperties, rendererIndex=0}
   {rendererClass=&NiagaraSpriteRendererProperties, rendererIndex=1}
   {rendererClass=&NiagaraSpriteRendererProperties, rendererIndex=2}
  ]
 }
}
summary {
 systemName=NS_Emote_PintPiano_Charge
 emitters [
  {emitterName=FrontCone, rendererClasses [&NiagaraMeshRendererProperties]}
  {emitterName=Front_Flares, rendererClasses [&NiagaraSpriteRendererProperties]}
  {emitterName=Flare, rendererClasses []}
  {emitterName=Thruster, rendererClasses [&NiagaraMeshRendererProperties, &NiagaraSpriteRendererProperties]}
  {emitterName=Boosters, rendererClasses [&NiagaraMeshRendererProperties, &NiagaraSpriteRendererProperties]}
 ]
 userVariables [
  User.bIsFrontEndPreview {type=NiagaraBool}
  User.LightVisibility {type=NiagaraFloat, defaultValue=1, description=""}
 ]
}
