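# Niagara system for the prop-spawn effect of the PintPiano main-player emote.
# It layers three emitters, SparklesLWE, AirglowLWE and SmokeLWE, whose LWE suffix suggests lightweight emitters.
# User floats BurstIndex and NormalizedNotifyProgress suggest control from an animation notify.
# The system uses SystemState and a set-variables module, with dynamic inputs for float comparison, bool-to-float conversion, random range and float-to-Vector2D.
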NiagaraSystem /BRCosmetics/Animation/Game/MainPlayer/Emotes/PintPiano/FX/NS_Emote_PintPiano_PropSpawn
dependencies {
 usedDataInterfaces []
 usedDynamicInputs [
  &Vector2DFromFloat
  &MakeCustomFloatFromBool
  &SetBoolByFloatComparison
  &RandomRangeFloat
 ]
 usedModules [
  &self:SystemSpawnScript.SystemScriptSource.SystemScriptGraph.NiagaraNodeAssignment_0.SetVariables_#1
  &SystemState
 ]
}
emitters {AirglowLWE {rendererClasses [], renderers []}, SmokeLWE {rendererClasses [], renderers []}, SparklesLWE {rendererClasses [], renderers []}}
summary {
 systemName=NS_Emote_PintPiano_PropSpawn
 emitters [{emitterName=SparklesLWE, rendererClasses []}, {emitterName=AirglowLWE, rendererClasses []}, {emitterName=SmokeLWE, rendererClasses []}]
 userVariables [
  User.BurstIndex {type=NiagaraFloat, defaultValue=0, description=""}
  User.NormalizedNotifyProgress {type=NiagaraFloat, defaultValue=0, description=""}
 ]
}
