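# Niagara effect for the teleport moment of the PintPiano emote, pairing fiery trails with a flash.
# FireTrails lays side-by-side partitioned ribbons along a rising curve driven by loop age, layering sprites, ribbons and lights tapered by execution index.
# NE_Flare adds a short burst of meshes, sprite and light, placed, scaled and delayed by shared System Explosion parameters.
# Some emitters, including refraction and smoke, are disabled.
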NiagaraSystem /BRCosmetics/Animation/Game/MainPlayer/Emotes/PintPiano/FX/NS_Emote_PintPiano_Teleport
dependencies {
 usedDataInterfaces [
  &NiagaraDataInterfaceCurve
  &NiagaraDataInterfaceVector2DCurve
  &NiagaraDataInterfaceArrayFloat
  &NiagaraDataInterfaceColorCurve
  &NiagaraDataInterfacePlatformSet
  &NiagaraDataInterfaceVectorCurve
 ]
 usedDynamicInputs [&SimulationPosition, &NormalizedAngleToDegrees]
 usedModules [
  &self:SystemSpawnScript.SystemScriptSource.SystemScriptGraph.NiagaraNodeAssignment_0.SetVariables_#1
  &self:SystemSpawnScript.SystemScriptSource.SystemScriptGraph.NiagaraNodeAssignment_1.SetVariables_#2
  &SystemState
 ]
}
emitters {
 FireTrails {
  emitterUpdateScript [
   &EmitterState(Life Cycle Mode=::NewEnumerator1, Loop Behavior=::NewEnumerator1, Loop Duration=0.3, UseLoopDelay=false)
   &SpawnBurst_Instantaneous(Spawn Group=1, Use Spawn Probability=false, UseLoopCountLimit=false)
   &SpawnRate(SpawnRate=150, Use Spawn Probability=false)
  ]
  particleSpawnScript [
   /SetVariables(Particles.VisibilityTag=&SpawnGroup()) {isSetParameters=true}
   &PartitionParticles(Write Partition to Ribbon ID=::NewEnumerator0)
   &InitializeParticle(
    Ribbon Width=10
    Ribbon Width Mode=::NewEnumerator1
    Sprite Size=&Multiply_Vector2D_ByFloat(Float=&RandomRangeFloat(Minimum=0.2, Maximum=0.2), Vector2D=<70 70>)
    Sprite Size Mode=::NewEnumerator4
    UsePositionOffset=false
    Write Material Random=true
   )
   &SystemLocation(
    MaskSpawnGroup=false
    Offset=&Add_Vector(
     A=<0 -50 10>
     B=&VectorFromCurve(
      VectorCurve=DataInterface() {propertyValues {
        XCurve {defaultValue=1, keys [{tangentMode=RCTM_User, time=0, value=1}, {tangentMode=RCTM_User, time=1, value=1}]}
        YCurve {
         defaultValue=1
         keys [
          {arriveTangent=1.0873, interpMode=RCIM_Cubic, leaveTangent=1.0873, tangentMode=RCTM_User, time=0, value=0}
          {arriveTangent=0.4204, interpMode=RCIM_Cubic, leaveTangent=0.4204, tangentMode=RCTM_User, time=1, value=1}
         ]
        }
        ZCurve {
         defaultValue=1
         keys [
          {arriveTangent=-0.4764, interpMode=RCIM_Cubic, leaveTangent=-0.4764, tangentMode=RCTM_User, time=0, value=0}
          {arriveTangent=1.1853, interpMode=RCIM_Cubic, leaveTangent=1.1853, tangentMode=RCTM_User, time=0.7597, value=0.1572}
          {arriveTangent=3.3818, interpMode=RCIM_Cubic, leaveTangent=3.3818, tangentMode=RCTM_User, time=1, value=0.7232}
         ]
        }
       }}
      CurveIndex=$Emitter.NormalizedLoopAge
      Scale Curve=&MaskVector(Mask Source=1, Mask Parameter=$Particles.PartitionParticles.Partition, Mask=0, Masked Vector=<-5 200 150>, Passthrough Vector=<5 200 150>)
     )
    )
    Offset Coordinate Space=::Simulation
   )
  ]
  particleUpdateScript [
   &ParticleState()
   /SetVariables(Particles.RibbonU1Override:NiagaraFloat=&Multiply_FloatByInt(Float=2, Integer=$Engine.ExecIndex)) {isSetParameters=true}
   /SetVariables(Particles.TopFacingRibbonU1Override:NiagaraFloat=&Multiply_FloatByInt(Float=1, Integer=$Engine.ExecIndex)) {isSetParameters=true}
   /SetVariables(
    Particles.FadeIn:NiagaraFloat=&FloatFromCurve(FloatCurve=DataInterface() {propertyValues {Curve {keys [<at 0 0>,<at 0.2 1>]}}}, CurveIndex=$Particles.Age)
    Particles.FadeOut:NiagaraFloat=&FloatFromCurve(FloatCurve=DataInterface() {propertyValues {Curve {keys [<at 0.6 1>,<at 1 0>]}}})
   ) {isSetParameters=true}
   &ScaleSpriteSize(
    Scale Sprite Size Mode=::NewEnumerator0
    Uniform Scale Factor=&Multiply_Float(
     A=&Multiply_Float(A=$Particles.FadeIn, B=$Particles.FadeOut)
     B=&FloatFromCurve(
      FloatCurve=DataInterface() {propertyValues {Curve {keys [
          {interpMode=RCIM_Cubic, tangentMode=RCTM_User, time=0, value=0.8}
          {interpMode=RCIM_Cubic, tangentMode=RCTM_User, time=0.1, value=1}
          {interpMode=RCIM_Cubic, tangentMode=RCTM_User, time=0.8, value=1}
          {interpMode=RCIM_Cubic, tangentMode=RCTM_User, time=1, value=0.8}
         ]}}}
      CurveIndex=&ReturnNormalizedExecIndex()
     )
    )
   )
   &ScaleRibbonWidth(Ribbon Width Scale=&Multiply_Float(
     A=&Multiply_Float(A=$Particles.FadeIn, B=$Particles.FadeOut)
     B=&FloatFromCurve(
      FloatCurve=DataInterface() {propertyValues {Curve {keys [{tangentMode=RCTM_User, time=0, value=0.5}, {tangentMode=RCTM_User, time=0.1, value=1}, <at 0.8 1>, <at 1 0.5>]}}}
      CurveIndex=&ReturnNormalizedExecIndex()
     )
    ))
   &SolveForcesAndVelocity(Clamp Velocity=false, Limit Acceleration=false)
   &SpriteFacingAndAlignment(
    Sprite Alignment=&Lerp_Vector(A=<0 0 1>, B=<0 -1 1>, Alpha=&ReturnNormalizedExecIndex())
    WriteAlignment=true
    WriteFacing=true
   )
   &DynamicMaterialParameters(
    Index 0 Param 1=&RandomRangeFloat(Minimum=0, Maximum=1)
    Index 1 Param 3=&Lerp_Float(A=2, B=0, Alpha=&Multiply_Float(A=$Particles.FadeIn, B=$Particles.FadeOut))
    Index 3 Param 1=&RandomRangeFloat(Minimum=0, Maximum=1)
    Module.Index0Float0Write=true
    Module.Index0Float1Write=true
    Module.Index0Float2Write=true
    Module.Index0Float3Write=true
    Module.Index1Float0Write=true
    Module.Index1Float1Write=true
    Module.Index1Float2Write=true
    Module.Index1Float3Write=true
    Module.Index3Float0Write=true
    Module.Index3Float1Write=true
    Module.Index3Float2Write=true
    Module.Index3Float3Write=true
    Write Parameter Index 1=::NewEnumerator1
    Write Parameter Index 3=::NewEnumerator1
   )
   /SetVariables() {isSetParameters=true}
   /SetVariables(Particles.GasRibbonWidth:NiagaraFloat=&Multiply_Float(A=$Particles.RibbonWidth, B=1)) {isSetParameters=true}
   &ScaleColor(
    Scale Alpha=&Multiply_Float(
     A=&FloatFromCurve(FloatCurve=DataInterface() {propertyValues {Curve {keys [<at 0 0>,<at 0.2 1>,<at 0.8 1>,<at 1 0>]}}})
     B=&FloatFromCurve(
      FloatCurve=DataInterface() {propertyValues {Curve {keys [{tangentMode=RCTM_User, time=0, value=0}, {tangentMode=RCTM_User, time=0.1, value=1}, <at 0.72 1>, <at 0.9311 0>]}}}
      CurveIndex=&ReturnNormalizedExecIndex()
      Scale Curve=&FNTimeOfDayScalar(Night Scale=0.6, Morning Scale=0.8, Evening Scale=0.8)
     )
    )
    ScaleA=true
    ScaleRGB=true
   )
   &FN_Sprite_Attributes(
    Color=&MultiplyLinearColor(
     Linear Color=$Particles.Color
     Scale Factor=&MakeVector4FromVectorAndFloat("Vector (RGB)"=<1 1 1>, "Float (Alpha)"=1)
    )
    Create Color Attribute=true
    Create Facing Attribute=true
    Create SpriteSize Attribute=true
    Facing:Vector3f=<0 1 0>
    SpriteSize=&Multiply_Vector2D(A=$Particles.SpriteSize, B=<0.7 2>)
    Write to Module Namespace=true
   )
   &FN_Sprite_Attributes(
    Alignment:Vector3f=&Lerp_Vector(A=<0 1 1>, B=<0 0 1>, Alpha=&ReturnNormalizedExecIndex())
    Color=&MultiplyLinearColor(
     Linear Color=$Particles.Color
     Scale Factor=&MakeVector4FromVectorAndFloat("Vector (RGB)"=<1 1 1>, "Float (Alpha)"=1)
    )
    Create Alignment Attribute=true
    Create Color Attribute=true
    Create Facing Attribute=true
    Create Position Attribute=true
    Create SpriteSize Attribute=true
    Facing:Vector3f=<0 0 1>
    Position:NiagaraPosition=&AddVectorToPosition(Position=$Particles.Position, Vector=<0 0 0>)
    SpriteSize=&Multiply_Vector2D_ByFloat(
     Float=&FloatFromCurve(
      FloatCurve=DataInterface() {propertyValues {Curve {keys [<at 0 0>,<at 0.2055 1>,<at 0.788 1>,<at 1 1>]}}}
      CurveIndex=&ReturnNormalizedExecIndex()
     )
     Vector2D=&Multiply_Vector2D(A=$Particles.SpriteSize, B=<0.6 1.3>)
    )
    Write to Module Namespace=true
   ) {module=FN_Sprite_Attributes001}
   /SetVariables(Particles.RibbonWidthFlame:NiagaraFloat=&Multiply_Float(A=$Particles.RibbonWidth, B=1.3)) {isSetParameters=true}
   /SetVariables(Particles.SideFacingSpriteSize:Vector2f=&Multiply_Vector2D(A=$Particles.SpriteSize, B=&RandomRangeVector2D(Minimum=<2 1>, Maximum=<2 2>))) {isSetParameters=true}
   &Light_Attributes(
    Color=1
    Falloff Exponent=1
    Light Exponent=3
    Light Position Offset:Vector3f=<0 50 35>
    Light Radius:NiagaraFloat=&FloatFromCurve(
     FloatCurve=DataInterface() {propertyValues {Curve {
        defaultValue=1
        keys [
         <at 0.0683 0 RCIM_Cubic>
         {arriveTangent=3.0514, interpMode=RCIM_Cubic, leaveTangent=3.0514, time=0.1, value=0.5255}
         {interpMode=RCIM_Cubic, tangentMode=RCTM_User, time=0.396, value=1}
         {arriveTangent=-2.9365, interpMode=RCIM_Cubic, leaveTangent=-2.9365, tangentMode=RCTM_User, time=1, value=0}
        ]
       }}}
     Scale Curve=120
    )
    Position=1
    Position to Offset:NiagaraPosition=$Engine.Emitter.SimulationPosition
    Radius=1
    Scale Alpha:NiagaraFloat=&FloatFromCurve(FloatCurve=DataInterface() {propertyValues {Curve {
        defaultValue=1
        keys [
         <at 0 0.5255 RCIM_Cubic>
         {interpMode=RCIM_Cubic, tangentMode=RCTM_User, time=0.296, value=1}
         {arriveTangent=-2.7423, interpMode=RCIM_Cubic, leaveTangent=-2.7423, tangentMode=RCTM_User, time=1, value=0}
        ]
       }}})
    Scale RGB:Vector3f=&MakeVectorFromLinearColorRGB(Color=<rgba 15 5.3546 0 1>)
    ScaleA=true
    ScaleRGB=true
    UsePositionOffset=true
   )
   /SetVariables(Particles.RibbonWidthFlat:NiagaraFloat=&Multiply_Float(A=$Particles.RibbonWidth, B=1.5)) {isSetParameters=true}
   /SetVariables(Particles.FlatFlameColor:LinearColor=&Multiply_LinearColorByFloat(
     Linear Color=$Particles.Color
     Float=&FloatFromCurve(
      FloatCurve=DataInterface() {propertyValues {Curve {
         defaultValue=1
         keys [
          {arriveTangent=5.0706, interpMode=RCIM_Cubic, leaveTangent=5.0706, tangentMode=RCTM_User, time=0, value=0}
          {interpMode=RCIM_Cubic, tangentMode=RCTM_User, time=0.5, value=1}
          {arriveTangent=-5.448, interpMode=RCIM_Cubic, leaveTangent=-5.448, tangentMode=RCTM_User, time=1, value=0}
         ]
        }}}
      CurveIndex=&ReturnNormalizedExecIndex()
      Scale Curve=0.5
     )
    )) {isSetParameters=true}
  ]
  rendererClasses [
   &NiagaraSpriteRendererProperties
   &NiagaraRibbonRendererProperties
   &NiagaraLightRendererProperties
  ]
  renderers [
   {rendererClass=&NiagaraSpriteRendererProperties}
   {rendererClass=&NiagaraRibbonRendererProperties, rendererIndex=1}
   {rendererClass=&NiagaraRibbonRendererProperties, rendererIndex=2}
   {rendererClass=&NiagaraSpriteRendererProperties, rendererIndex=3}
   {rendererClass=&NiagaraLightRendererProperties, rendererIndex=4}
   {rendererClass=&NiagaraRibbonRendererProperties, rendererIndex=5}
  ]
 }
 LWE_Refract {bEnabled=false, rendererClasses [], renderers []}
 Main_Smoke001 {
  bEnabled=false
  emitterSpawnScript [
   /SetVariables(
    Emitter.Global Color:LinearColor=<rgba 4 0.349 0 1>
    Emitter.Global Scale:NiagaraFloat=&Multiply_Float(A=0.15, B=&FN_FrontEndFloatModifier(Gameplay=1, Front End=$System.FrontEndScale))
   ) {isSetParameters=true}
   /SetVariables(
    Emitter.Color2:LinearColor=&MultiplyLinearColor(Linear Color=$Emitter.Global Color, Scale Factor=<2 1 1 1>)
    Emitter.Depth Fade:NiagaraFloat=&Multiply_Float(A=125, B=$Emitter.Global Scale)
    Emitter.Dissolve Noise Power=1.1
    Emitter.Dissolve Noise Scale:LinearColor=&MultiplyLinearColor(
     Linear Color=/Niagara/DynamicInputs/TypeConversions/MakeLinearColorFromFloat(Value:NiagaraFloat=&FloatFromCurve(FloatCurve=DataInterface() {propertyValues {Curve {keys [<at 0 2>,<at 1 1>,<at 2 0.5>]}}}, CurveIndex=$Emitter.Global Scale))
     Scale Factor=<0.0008 0.0008 0.0009 1>
    )
    Emitter.Near Camera Fade:NiagaraFloat=&Multiply_Float(A=400, B=$Emitter.Global Scale)
   ) {isSetParameters=true}
   /SetVariables(Emitter.Dissolve Scale Adjsutment:NiagaraFloat=&FloatFromCurve(
     FloatCurve=DataInterface() {propertyValues {Curve {keys [<at 0 3>,<at 1 1>,<at 5 0.5>,<at 20 0.4>,<at 100 0.4>]}}}
     CurveIndex=$Emitter.Global Scale
     Scale Curve=$Emitter.Global Scale
    )) {isSetParameters=true}
  ]
  emitterUpdateScript [
   &EmitterState(Life Cycle Mode=::NewEnumerator1, Loop Behavior=::NewEnumerator1, Loop Delay=$System.ExplosionDelay, Loop Duration=0.627, UseLoopDelay=true)
   &SpawnBurst_Instantaneous(Spawn Group=1, Spawn Time=0.01, Use Spawn Probability=false, UseLoopCountLimit=false) {enabled=false, module=SpawnBurst_Instantaneous002}
   &SpawnBurst_Instantaneous(Spawn Count=8, Use Spawn Probability=false, UseLoopCountLimit=false) {module=SpawnBurst_Instantaneous001}
  ]
  particleSpawnScript [
   &InitializeParticle(
    Color=&MaskLinearColorBySpawnGroup(Masked Linear Color=<rgba 0.9048 0.9062 0.9048 2>, Passthrough Linear Color=<rgba 0.4323 0.4252 0.4161 1.7>, Spawn Group Mask=1)
    Lifetime=&MaskFloatBySpawnGroup(
     Masked Float=1
     Passthrough Float=&SelectFloatFromArray(
      Array Sampling Mode=::NewEnumerator0
      Direct Array Index=?
      Float Selection Array=DataInterface() {propertyValues {FloatData [1,1.2]}}
      Random Seed=0
     )
     Spawn Group Mask=1
    )
    Mesh Scale Mode=::NewEnumerator3
    Mesh Uniform Scale=&Multiply_Float(A=$System.ExplosionScale, B=&RandomRangeFloat(Minimum=4, Maximum=5))
    Position Offset=$System.ExplosionPosition
    UsePositionOffset=true
    Write Material Random=true
   )
   &ShapeLocation(MaskSpawnGroup=false, Offset Mode=::NewEnumerator0, Override Seed=false, Sphere Radius=1, UseOwnerScale=true)
   &AddVelocity(
    Module.Use Falloff Distance=false
    Origin Offset=&Multiply_VectorByFloat(Vector=&RandomRangeVector(Minimum=<0 0 0>, Maximum=<0 0 -5>), Float=$Emitter.Global Scale)
    Override Seed=false
    Velocity Mode=::NewEnumerator1
    Velocity Origin:Vector3f=$System.ExplosionPosition
    Velocity Speed=&Multiply_Float(
     A=&MaskFloatBySpawnGroup(Masked Float=1400, Passthrough Float=&RandomRangeFloat(Minimum=900, Maximum=1200), Spawn Group Mask=1)
     B=$Emitter.Global Scale
    )
   )
   &CameraOffset(Camera Offset Amount=&Multiply_Float(A=&MaskFloatBySpawnGroup(Masked Float=20, Passthrough Float=15, Spawn Group Mask=1), B=$Emitter.Global Scale))
   /SetVariables(Particles.Scale=&Multiply_VectorByFloat(Vector=$Particles.Scale, Float=$Emitter.Global Scale)) {isSetParameters=true}
  ]
  particleUpdateScript [
   &ParticleState()
   &GravityForce(Gravity=&Multiply_VectorByFloat(
     Vector=&MaskVectorBySpawnGroup(
      Masked Vector=<0 0 500>
      Passthrough Vector=&RandomRangeVector(Minimum=<0 0 400>, Maximum=<1 1 600>)
      Spawn Group Mask=1
     )
     Float=$Emitter.Global Scale
    ))
   &Drag(Drag=&RandomRangeFloat(Minimum=6, Maximum=8), Use Linear Drag=true, Use Rotational Drag=false)
   &DynamicMaterialParameters(
    Index 0 Param 1=&Multiply_Float(
     A=&MaskFloatBySpawnGroup(
      Masked Float=-150
      Passthrough Float=&RandomRangeFloat(Minimum=-90, Maximum=-120)
      Spawn Group Mask=&MakeIntFromFloat(Float=1)
     )
     B=$Emitter.Global Scale
    )
    Index 0 Param 2=&FloatFromCurve(
     FloatCurve=DataInterface() {propertyValues {Curve {keys [
         {tangentMode=RCTM_User, time=0, value=1}
         {arriveTangent=-0.0405, interpMode=RCIM_Cubic, leaveTangent=-0.0405, tangentMode=RCTM_User, time=0.4, value=1}
         {arriveTangent=-1.6592, interpMode=RCIM_Cubic, leaveTangent=-1.6592, tangentMode=RCTM_User, time=1, value=0.4}
        ]}}}
     Scale Curve=0.5
    )
    Index 0 Param 3=&FloatFromCurve(
     FloatCurve=DataInterface() {propertyValues {Curve {keys [
         {arriveTangent=1.4725, interpMode=RCIM_Cubic, leaveTangent=1.4725, tangentMode=RCTM_User, time=0, value=0}
         {tangentMode=RCTM_User, time=1, value=1}
        ]}}}
     Scale Curve=&Multiply_Float(
      A=&MaskFloatBySpawnGroup(Masked Float=350, Passthrough Float=&RandomRangeFloat(Minimum=300, Maximum=400), Spawn Group Mask=1)
      B=$Emitter.Global Scale
     )
    )
    Index 0 Param 4=&MaskFloatBySpawnGroup(
     Masked Float=&FloatFromCurve(
      FloatCurve=DataInterface() {propertyValues {Curve {keys [
          {arriveTangent=0.1175, interpMode=RCIM_Cubic, leaveTangent=0.6205, tangentMode=RCTM_Break, time=0.22, value=-2}
          {arriveTangent=6.3488, interpMode=RCIM_Cubic, leaveTangent=6.3488, tangentMode=RCTM_User, time=0.6, value=-0.4172}
          {arriveTangent=36.7876, interpMode=RCIM_Cubic, leaveTangent=14.4841, tangentMode=RCTM_Break, time=1, value=7.3}
         ]}}}
      Scale Curve=0.2
     )
     Passthrough Float=&FloatFromCurve(
      FloatCurve=DataInterface() {propertyValues {Curve {keys [
          {arriveTangent=0.1175, interpMode=RCIM_Cubic, leaveTangent=5.607, tangentMode=RCTM_Break, time=0.1, value=-2.5}
          {arriveTangent=30.0332, interpMode=RCIM_Cubic, leaveTangent=14.4841, tangentMode=RCTM_Break, time=1, value=9}
         ]}}}
      Scale Curve=0.2
     )
     Spawn Group Mask=1
    )
    Index 1 Param 1=&FloatFromCurve(
     FloatCurve=DataInterface() {propertyValues {Curve {keys [
         {arriveTangent=0.5348, interpMode=RCIM_Cubic, leaveTangent=0.5348, tangentMode=RCTM_User, time=0.05, value=0.02}
         {arriveTangent=39.7786, interpMode=RCIM_Cubic, leaveTangent=39.7786, tangentMode=RCTM_User, time=0.4, value=8}
        ]}}}
     Scale Curve=&MaskFloatBySpawnGroup(Masked Float=85, Passthrough Float=&RandomRangeFloat(Minimum=55, Maximum=55), Spawn Group Mask=1)
    )
    Index 1 Param 2=&Divide_Float(
     A=&MaskFloatBySpawnGroup(
      Masked Float=0.01
      Passthrough Float=&RandomRangeFloat(Minimum=0.0089, Maximum=0.015)
      Spawn Group Mask=1
     )
     B=$Emitter.Global Scale
    )
    Index 1 Param 3=&FloatFromCurve(
     FloatCurve=DataInterface() {propertyValues {Curve {keys [{interpMode=RCIM_Cubic, tangentMode=RCTM_User, time=0.4, value=0.5}, <at 0.8 0.1>]}}}
     Scale Curve=5
    )
    Index 1 Param 4=&MaskFloatBySpawnGroup(Masked Float=2, Passthrough Float=1, Spawn Group Mask=1)
    Index 2 Param 1=&FloatFromCurve(FloatCurve=DataInterface() {propertyValues {Curve {keys [
         {arriveTangent=0.5431, interpMode=RCIM_Cubic, leaveTangent=0.5431, tangentMode=RCTM_User, time=0.3, value=0}
         {arriveTangent=2.1449, interpMode=RCIM_Cubic, leaveTangent=2.1449, tangentMode=RCTM_User, time=1, value=1}
        ]}}})
    Index 2 Param 2=&FloatFromCurve(
     FloatCurve=DataInterface() {propertyValues {Curve {keys [
         {interpMode=RCIM_Cubic, tangentMode=RCTM_User, time=0.3, value=0}
         {arriveTangent=2.4333, interpMode=RCIM_Cubic, leaveTangent=2.4333, tangentMode=RCTM_User, time=1, value=1}
        ]}}}
     Scale Curve=&Multiply_Float(
      A=0.1
      B=&FloatFromCurve(
       FloatCurve=DataInterface() {propertyValues {Curve {keys [<at 0 1>,<at 1 1>,<at 10 1>,<at 20 1>,<at 100 0.2>]}}}
       CurveIndex=$Emitter.Global Scale
      )
     )
    )
    Index 2 Param 3=&MaskFloatBySpawnGroup(
     Masked Float=&FloatFromCurve(
      FloatCurve=DataInterface() {propertyValues {Curve {keys [
          {arriveTangent=-6.4028, interpMode=RCIM_Cubic, leaveTangent=-6.4028, tangentMode=RCTM_User, time=0.1, value=1}
          {arriveTangent=-0.0065, interpMode=RCIM_Cubic, leaveTangent=-0.0065, tangentMode=RCTM_User, time=0.4, value=0}
         ]}}}
      Scale Curve=5
     )
     Passthrough Float=&FloatFromCurve(
      FloatCurve=DataInterface() {propertyValues {Curve {keys [
          {arriveTangent=-3.4318, interpMode=RCIM_Cubic, leaveTangent=-3.4318, tangentMode=RCTM_User, time=0.1, value=1}
          {arriveTangent=-0.7527, interpMode=RCIM_Cubic, leaveTangent=-0.7527, tangentMode=RCTM_User, time=0.35, value=0}
         ]}}}
      Scale Curve=5
     )
     Spawn Group Mask=1
    )
    Index 2 Param 4=&FloatFromCurve(
     FloatCurve=DataInterface() {propertyValues {Curve {keys [
         {arriveTangent=-16.6453, interpMode=RCIM_Cubic, leaveTangent=-16.6453, tangentMode=RCTM_User, time=0, value=64}
         {arriveTangent=-87.1792, interpMode=RCIM_Cubic, leaveTangent=-87.1792, tangentMode=RCTM_User, time=0.1, value=24}
         {arriveTangent=-52.5, interpMode=RCIM_Cubic, leaveTangent=-52.5, time=0.2, value=12}
         {arriveTangent=-7.0531, interpMode=RCIM_Cubic, leaveTangent=-7.0531, tangentMode=RCTM_User, time=0.5, value=3}
         {arriveTangent=-0.6909, interpMode=RCIM_Cubic, leaveTangent=-0.6909, tangentMode=RCTM_User, time=1, value=2}
         <at 2 1>
         <at 3 0.3>
         <at 10 0.1>
         <at 20 0.05>
         <at 50 0.02>
         <at 100 0.01>
        ]}}}
     CurveIndex=$Emitter.Global Scale
     Scale Curve=0.6
    )
    Index 3 Param 1=&FloatFromCurve(
     FloatCurve=DataInterface() {propertyValues {Curve {keys [
         {arriveTangent=2.2353, interpMode=RCIM_Cubic, leaveTangent=2.2353, tangentMode=RCTM_User, time=0, value=0.2}
         {arriveTangent=1.4416, interpMode=RCIM_Cubic, leaveTangent=1.4416, tangentMode=RCTM_User, time=0.2, value=0.5}
         {interpMode=RCIM_Cubic, tangentMode=RCTM_User, time=0.8, value=1.25}
        ]}}}
     Scale Curve=-1
    )
    Index 3 Param 2=&FloatFromCurve(
     FloatCurve=DataInterface() {propertyValues {Curve {keys [
         {arriveTangent=5.1563, interpMode=RCIM_Cubic, leaveTangent=5.1563, tangentMode=RCTM_User, time=0, value=0}
         {arriveTangent=1.4416, interpMode=RCIM_Cubic, leaveTangent=1.4416, tangentMode=RCTM_User, time=0.1699, value=0.4}
         {interpMode=RCIM_Cubic, tangentMode=RCTM_User, time=1, value=1.1}
        ]}}}
     Scale Curve=-0.1
    )
    Index 3 Param 3=&FloatFromCurve(FloatCurve=DataInterface() {propertyValues {Curve {keys [
         {arriveTangent=-11.5239, interpMode=RCIM_Cubic, leaveTangent=-11.5239, tangentMode=RCTM_User, time=0.15, value=40}
         {arriveTangent=-88.9092, interpMode=RCIM_Cubic, leaveTangent=-88.9092, tangentMode=RCTM_User, time=0.3, value=25}
         {arriveTangent=-43.7054, interpMode=RCIM_Cubic, leaveTangent=-43.7054, tangentMode=RCTM_User, time=0.6, value=0}
         {interpMode=RCIM_Cubic, tangentMode=RCTM_User, time=0.7, value=0}
         {interpMode=RCIM_Cubic, tangentMode=RCTM_User, time=0.8, value=0}
        ]}}})
    Index 3 Param 4=&FloatFromCurve(
     FloatCurve=DataInterface() {propertyValues {Curve {keys [{tangentMode=RCTM_User, time=0, value=0}, {tangentMode=RCTM_User, time=1, value=1}]}}}
     Scale Curve=-250
    )
    Module.Index0Param0Write=true
    Module.Index0Param1Write=true
    Module.Index0Param2Write=true
    Module.Index0Param3Write=true
    Module.Index1Param0Write=true
    Module.Index1Param1Write=true
    Module.Index1Param2Write=true
    Module.Index1Param3Write=true
    Module.Index2Param0Write=true
    Module.Index2Param1Write=true
    Module.Index2Param2Write=true
    Module.Index2Param3Write=true
    Module.Index3Param0Write=true
    Module.Index3Param1Write=true
    Module.Index3Param2Write=true
    Module.Index3Param3Write=true
    Write Parameter Index 1=true
    Write Parameter Index 2=true
    Write Parameter Index 3=true
   ) {module=DynamicMaterialParameters001}
   &ScaleColor(
    Scale Alpha=&FloatFromCurve(FloatCurve=DataInterface() {propertyValues {Curve {keys [<at 0 1>,<at 1 0>]}}})
    Scale RGB=&MakeVectorFromLinearColorRGB(Color=&MultiplyLinearColor(
      Linear Color=&ColorFromCurve(ColorCurve=DataInterface() {propertyValues {
         AlphaCurve {keys [<at 0.0022 1>,<at 0.994 0>]}
         BlueCurve {keys [<at 0.1888 0.6042>,<at 0.4977 0.4497>]}
         GreenCurve {keys [<at 0.1888 0.6042>,<at 0.4977 0.4497>]}
         RedCurve {keys [<at 0.1888 0.6042>,<at 0.4977 0.4497>]}
        }})
      Scale Factor=&MakeVector4_FromLinearColor(LinearColor=&ColorFromCurve(
        ColorCurve=DataInterface() {propertyValues {
          AlphaCurve {keys [<at 0 1>,<at 0.125 1>,<at 0.25 1>,<at 0.375 1>,<at 0.5 1>,<at 0.625 1>,<at 0.75 1>,<at 0.875 1>,<at 1 1>]}
          BlueCurve {keys [<at 0.125 0.7835>,<at 0.166 0.6172>,<at 0.208 0.5395>,<at 0.25 0.6172>,<at 0.291 0.6939>,<at 0.333 0.7379>,<at 0.375 0.7835>,<at 0.416 0.7991>,<at 0.458 0.8148>,<at 0.5 0.8308>,<at 0.541 0.8148>,<at 0.583 0.7991>,<at 0.625 0.7991>,<at 0.666 0.7529>,<at 0.708 0.7084>,<at 0.75 0.5711>,<at 0.791 0.4735>,<at 0.833 0.3231>,<at 0.8794 0.7835>]}
          GreenCurve {keys [<at 0.125 0.3118>,<at 0.166 0.5333>,<at 0.208 0.5271>,<at 0.25 0.6376>,<at 0.291 0.7529>,<at 0.333 0.7835>,<at 0.375 0.8148>,<at 0.416 0.7991>,<at 0.458 0.8148>,<at 0.5 0.8469>,<at 0.541 0.7991>,<at 0.583 0.7991>,<at 0.625 0.8148>,<at 0.666 0.7991>,<at 0.708 0.7529>,<at 0.75 0.5711>,<at 0.791 0.5029>,<at 0.833 0.5029>,<at 0.8794 0.3118>]}
          RedCurve {keys [<at 0.125 0.1252>,<at 0.166 0.7084>,<at 0.208 0.6867>,<at 0.25 0.8148>,<at 0.291 0.8879>,<at 0.333 0.8963>,<at 0.375 0.9131>,<at 0.416 0.8714>,<at 0.458 0.8714>,<at 0.5 0.9131>,<at 0.541 0.8632>,<at 0.583 0.8714>,<at 0.625 0.9131>,<at 0.666 0.8963>,<at 0.708 0.8714>,<at 0.75 0.7529>,<at 0.791 0.7529>,<at 0.833 0.8469>,<at 0.8794 0.1252>]}
         }}
        CurveIndex=Linked() {linkedVariable {name=NPC.FortniteNPC.TimeOfDay_0-1, type=NiagaraFloat}}
       ))
     ))
    ScaleA=true
    ScaleRGB=true
   )
   &CurlNoiseForce(
    Noise Frequency=&Divide_Float(A=4, B=$Emitter.Global Scale)
    Noise Quality / Cost=::NewEnumerator0
    Noise Strength=&Multiply_Float(A=&MaskFloatBySpawnGroup(Masked Float=0, Passthrough Float=1000, Spawn Group Mask=1), B=$Emitter.Global Scale)
    Pan Noise=false
   )
   &PointAttractionForce(
    Attraction Radius=500
    Attraction Strength=&FloatFromCurve(FloatCurve=DataInterface() {propertyValues {Curve {
        defaultValue=-500
        keys [
         {arriveTangent=-119.0229, interpMode=RCIM_Cubic, leaveTangent=-119.0229, tangentMode=RCTM_User, time=0.2981, value=-10}
         {arriveTangent=2287.5398, interpMode=RCIM_Cubic, leaveTangent=2287.5398, tangentMode=RCTM_User, time=0.7, value=400}
        ]
       }}})
    Attractor Position:Vector3f=$System.ExplosionPosition
    Kill Radius:NiagaraFloat=&FloatFromCurve(FloatCurve=DataInterface() {propertyValues {Curve {
        defaultValue=-500
        keys [
         {arriveTangent=2.4671, interpMode=RCIM_Cubic, leaveTangent=2.4671, tangentMode=RCTM_User, time=0, value=0}
         {arriveTangent=18.5105, interpMode=RCIM_Cubic, leaveTangent=18.5105, tangentMode=RCTM_User, time=0.7, value=15}
        ]
       }}})
    Kill Within Radius=true
    Module.Use Falloff=true
    UseAttractorOffset=false
   )
   &VortexForce(UseOriginOffset=false, Vortex Force Amount=473.4479, Vortex Origin:Vector3f=$System.ExplosionPosition)
   &SolveForcesAndVelocity(Clamp Velocity=false, Limit Acceleration=false)
   &ScaleMeshSize(Scale Factor=&VectorFromFloat(Value=&FloatFromCurve(
      FloatCurve=DataInterface() {propertyValues {Curve {
         defaultValue=1
         keys [
          {arriveTangent=9.0725, interpMode=RCIM_Cubic, leaveTangent=9.0725, tangentMode=RCTM_User, time=0, value=0}
          {arriveTangentWeight=0.1274, interpMode=RCIM_Cubic, leaveTangentWeight=0.2024, tangentMode=RCTM_User, time=0.1408, value=0.9847}
          {arriveTangentWeight=0.1134, interpMode=RCIM_Cubic, leaveTangentWeight=0.3223, tangentMode=RCTM_User, time=0.3349, value=0.9786}
          {arriveTangent=-1.4713, interpMode=RCIM_Cubic, leaveTangent=-1.4713, time=0.4685, value=0.3015}
          {arriveTangent=-0.0405, interpMode=RCIM_Cubic, leaveTangent=-0.0405, tangentMode=RCTM_User, time=1, value=0}
         ]
        }}}
      Scale Curve=0.8
     )))
  ]
  rendererClasses [&NiagaraMeshRendererProperties]
  renderers [{rendererClass=&NiagaraMeshRendererProperties, rendererIndex=0}]
 }
 NE_Flare {
  emitterUpdateScript [
   &EmitterState(Life Cycle Mode=::NewEnumerator1, Loop Behavior=::NewEnumerator1, Loop Delay=$System.ExplosionDelay, Loop Duration=0.4, UseLoopDelay=true)
   &SpawnBurst_Instantaneous(Spawn Count=1, Use Spawn Probability=false)
   &SpawnBurst_Instantaneous(Spawn Count=1, Spawn Group=1, Use Spawn Probability=false) {module=SpawnBurst_Instantaneous001}
   &SpawnBurst_Instantaneous(Spawn Count=1, Spawn Group=2, Use Spawn Probability=false) {module=SpawnBurst_Instantaneous002}
   &SpawnBurst_Instantaneous(Spawn Count=1, Spawn Group=3, Use Spawn Probability=false) {module=SpawnBurst_Instantaneous003}
  ]
  particleSpawnScript [
   &InitializeParticle(
    Color=<rgba 1.8184 1.8286 5 1>
    Lifetime=0.3
    Mass Mode=::NewEnumerator1
    Mesh Scale=&VectorFromFloat(Value=&Multiply_Float(A=&MaskFloatBySpawnGroup(Masked Float=3, Passthrough Float=3), B=$System.ExplosionScale))
    Mesh Scale Mode=::NewEnumerator4
    Position Offset=$System.ExplosionPosition
    Sprite Size=&Multiply_Vector2D_ByFloat(Float=$System.ExplosionScale, Vector2D=<2 2>)
    Sprite Size Mode=::NewEnumerator4
    UsePositionOffset=true
    Write Material Random=true
   ) {module=InitializeParticle001}
   /SetVariables(Particles.VisibilityTag=&SpawnGroup()) {isSetParameters=true, module=SetVariables_#3}
   &InitialMeshOrientation(Orientation Axis=<0 0 1>, Orientation Vector=<0 0 1>, Rotation=<0 0.25 0>, Use Orientation Vector=true, Use Rotation Vector=true)
   &InitialMeshOrientation(
    Orientation Axis=<0 0 0>
    Orientation Vector=<0 0 0>
    Rotation=&RandomRangeVector(Minimum=<0 0 0>, Maximum=<1 0 0>)
    Use Orientation Vector=false
    Use Rotation Vector=true
   ) {module=InitialMeshOrientation001}
   /SetVariables(Particles.BigFlareOrient:Quat4f=&OrientMeshToVector(Look At Direction=<0 0 0>, Reference Axis Vector=<1 0 0>)) {isSetParameters=true}
   /SetVariables(Particles.BigFlareDynmat:Vector4f=<0.8 0 0 0>) {isSetParameters=true}
  ]
  particleUpdateScript [
   &ParticleState()
   &ScaleMeshSize(Scale Factor=&VectorFromFloat(Value=&MaskFloatBySpawnGroup(
      Masked Float=&FloatFromCurve(FloatCurve=DataInterface() {propertyValues {Curve {keys [
           {arriveTangent=4.4038, interpMode=RCIM_Cubic, leaveTangent=4.4038, tangentMode=RCTM_User, time=0, value=0.552}
           {arriveTangent=1.3811, interpMode=RCIM_Cubic, leaveTangent=1.3811, time=0.0575, value=1.042}
           {arriveTangent=-1.5561, interpMode=RCIM_Cubic, leaveTangent=-1.5561, tangentMode=RCTM_User, time=0.3441, value=1.0272}
           {arriveTangent=-36.7498, interpMode=RCIM_Cubic, leaveTangent=-36.7498, tangentMode=RCTM_User, time=0.4094, value=0}
          ]}}})
      Passthrough Float=&FloatFromCurve(
       FloatCurve=DataInterface() {propertyValues {Curve {keys [
           {interpMode=RCIM_Cubic, leaveTangent=1.9146, tangentMode=RCTM_Break, time=0, value=0.8157}
           {arriveTangent=3.2296, interpMode=RCIM_Cubic, leaveTangent=3.2296, time=0.068, value=1.56}
           {arriveTangent=0.4214, interpMode=RCIM_Cubic, leaveTangent=0.4214, time=0.367, value=2.0009}
           {arriveTangent=-1.4303, interpMode=RCIM_Cubic, leaveTangent=-1.4303, tangentMode=RCTM_User, time=0.8292, value=1.8808}
           {arriveTangent=-20.6168, interpMode=RCIM_Cubic, tangentMode=RCTM_Break, time=0.9199, value=0}
          ]}}}
       Scale Curve=0.3
      )
     )))
   &ScaleColor(
    Scale Alpha=&MaskFloatBySpawnGroup(
     Masked Float=1
     Passthrough Float=&FloatFromCurve(
      FloatCurve=DataInterface() {propertyValues {Curve {keys [
          {interpMode=RCIM_Cubic, tangentMode=RCTM_User, time=0, value=1}
          {arriveTangent=-1, interpMode=RCIM_Cubic, leaveTangent=-1, time=0.9679, value=0.9259}
          {arriveTangent=-2.6886, interpMode=RCIM_Cubic, leaveTangent=-2.6886, tangentMode=RCTM_User, time=1, value=0}
         ]}}}
      Scale Curve=0.5
     )
    )
    Scale RGB=&MaskVectorBySpawnGroup(
     Masked Vector=&VectorFromFloat(Value=&FloatFromCurve(FloatCurve=DataInterface() {propertyValues {Curve {keys [<at 0 0.7>,<at 0.9015 0.7>,<at 1 0>]}}}, Scale Curve=0.2))
     Passthrough Vector=&Multiply_Vector(
      A=&MakeVectorFromLinearColorRGB(Color=&ColorFromCurve(ColorCurve=DataInterface() {propertyValues {
          AlphaCurve {keys [<at 0 1>,<at 0.4696 1>,<at 1 0>]}
          BlueCurve {keys [<at 0 1>,<at 0.2681 1>,<at 0.4696 0.6979>,<at 1 0.3698>]}
          GreenCurve {keys [<at 0 0.6711>,<at 0.2681 0.6546>,<at 0.4696 0.3778>,<at 1 0.0702>]}
          RedCurve {keys [<at 0 0.7284>,<at 0.2681 0.7>,<at 0.4696 0.436>,<at 1 0.0763>]}
         }}))
      B=&VectorFromFloat(Value=&FloatFromCurve(
        FloatCurve=DataInterface() {propertyValues {Curve {keys [
            {interpMode=RCIM_Cubic, tangentMode=RCTM_Break, time=0, value=10}
            {arriveTangent=-6.6072, interpMode=RCIM_Cubic, leaveTangent=-6.6072, tangentMode=RCTM_User, time=0.9121, value=9.2635}
            {interpMode=RCIM_Cubic, tangentMode=RCTM_Break, time=1, value=1}
           ]}}}
        Scale Curve=0.1
       ))
     )
    )
    ScaleA=true
    ScaleRGB=true
   )
   &DynamicMaterialParameters(
    Index 0 Param 1=&FloatFromCurve(FloatCurve=DataInterface() {propertyValues {Curve {keys [{arriveTangent=-2.1015, interpMode=RCIM_Cubic, leaveTangent=-2.1015, tangentMode=RCTM_User, time=0, value=0.5}, <at 1 -0.15>]}}})
    Index 0 Param 2=&FloatFromCurve(FloatCurve=DataInterface() {propertyValues {Curve {keys [
         {arriveTangent=-0.1313, interpMode=RCIM_Cubic, leaveTangent=-0.1313, tangentMode=RCTM_User, time=0, value=1}
         {arriveTangent=-0.1, interpMode=RCIM_Cubic, leaveTangent=-0.1, time=0.9745, value=1.0003}
         <at 1 0.9>
        ]}}})
    Index 0 Param 3=&FloatFromCurve(FloatCurve=DataInterface() {propertyValues {Curve {keys [
         {interpMode=RCIM_Cubic, tangentMode=RCTM_Break, time=0, value=0}
         {arriveTangent=0.4478, interpMode=RCIM_Cubic, leaveTangent=0.4478, tangentMode=RCTM_User, time=0.15, value=1}
         {interpMode=RCIM_Cubic, tangentMode=RCTM_Break, time=1, value=1.2}
        ]}}})
    Index 3 Param 1=1
    Module.Index0Float0Write=true
    Module.Index0Float1Write=true
    Module.Index0Float2Write=true
    Module.Index0Float3Write=true
    Module.Index3Float0Write=true
    Module.Index3Float1Write=true
    Module.Index3Float2Write=true
    Module.Index3Float3Write=true
    Write Parameter Index 3=::NewEnumerator1
   )
   /SetVariables(Particles.BigFlareScale:Vector3f=&Multiply_VectorByFloat(Vector=$Particles.Scale, Float=0.15)) {isSetParameters=true}
   /SetVariables(Particles.BigFlareColor:LinearColor=&MultiplyLinearColor(
     Linear Color=<rgba 50 20.4059 4.2605 1>
     Scale Factor=&MakeFloat4FromVectorAndFloat(
      "Vector (RGB)"=&VectorFromFloat(Value=1)
      "Float (Alpha)"=&FloatFromCurve(
       FloatCurve=DataInterface() {propertyValues {Curve {keys [<at 0.4 1 RCIM_Cubic>,<at 1 0>]}}}
       Scale Curve=&FNTimeOfDayScalar(Night Scale=0.7, Morning Scale=0.8, Evening Scale=0.8)
      )
     )
    )) {isSetParameters=true}
   &FN_Light_Attributes(
    Color=$Particles.Color
    Create Color Attribute=true
    Create Exponent Attribute=true
    Create Position Attribute=true
    Create Radius Attribute=true
    Exponent=1
    Position=$System.ExplosionPosition
    Radius=&Multiply_Float(
     A=200
     B=&FloatFromCurve(
      FloatCurve=DataInterface() {propertyValues {Curve {keys [
          {arriveTangent=1.4112, interpMode=RCIM_Cubic, leaveTangent=1.4112, tangentMode=RCTM_User, time=0, value=1}
          {interpMode=RCIM_Cubic, tangentMode=RCTM_User, time=1, value=0}
         ]}}}
      Scale Curve=$System.ExplosionScale
     )
    )
    Scale Color=&FloatFromCurve(
     FloatCurve=DataInterface() {propertyValues {Curve {keys [
         {interpMode=RCIM_Cubic, tangentMode=RCTM_User, time=0, value=1}
         {arriveTangent=-3.1925, interpMode=RCIM_Cubic, leaveTangent=-3.1925, tangentMode=RCTM_User, time=1, value=0}
        ]}}}
     Scale Curve=20
    )
   )
   &FN_Mesh_Attributes(
    Color=&MultiplyLinearColor(
     Linear Color=<rgba 0.3268 0.1962 2 2>
     Scale Factor=&MakeVector4FromVectorAndFloat(
      "Vector (RGB)"=&VectorFromFloat(Value=&FloatFromCurve(FloatCurve=DataInterface() {propertyValues {Curve {keys [<at 0 1>,<at 0.7 0>]}}}))
      "Float (Alpha)"=&FloatFromCurve(FloatCurve=DataInterface() {propertyValues {Curve {keys [
           {interpMode=RCIM_Cubic, tangentMode=RCTM_User, time=0.0045, value=0.9397}
           {arriveTangent=-1.172, interpMode=RCIM_Cubic, leaveTangent=-1.172, tangentMode=RCTM_User, time=1.0011, value=0.002}
          ]}}})
     )
    )
    Create Color Attribute=true
    Create Scale Attribute=true
    Scale=&VectorFromFloat(Value=&Multiply_Float(
      A=11
      B=&Multiply_Float(
       A=$System.ExplosionScale
       B=&FloatFromCurve(FloatCurve=DataInterface() {propertyValues {Curve {keys [
            {arriveTangent=2.2061, interpMode=RCIM_Cubic, leaveTangent=2.2061, tangentMode=RCTM_User, time=0, value=1}
            {interpMode=RCIM_Cubic, tangentMode=RCTM_User, time=1, value=0}
           ]}}})
      )
     ))
    Write to Module Namespace=true
   )
   /SetVariables(Particles.RefracSphereDynMat:Vector4f=&MakeVector4(
     X=&Lerp_Float(
      A=1
      B=&FloatFromCurve(FloatCurve=DataInterface() {propertyValues {Curve {keys [<at 0 1>,<at 0.3172 1>,<at 1 1>]}}})
      Alpha=&FN_PickFloatDependOnPlatform(TargetPlatformActive=1, TargetPlatformNoActive=0)
     )
     Y=&FloatFromCurve(FloatCurve=DataInterface() {propertyValues {Curve {keys [<at 0 15>,<at 1 0>]}}})
     Z=0
     W=0
    )) {isSetParameters=true, module=SetVariables_#4}
   &FN_Sprite_Attributes(
    Camera Offset=20
    Color=&Multiply_LinearColorByFloat(Linear Color=$Particles.Color, Float=5)
    Create Camera Offset Attribute=true
    Create Color Attribute=true
    Create SpriteSize Attribute=true
    SpriteSize=&Multiply_Vector2D_ByFloat(
     Float=&Multiply_Float(
      A=&Multiply_Float(
       A=280
       B=&FloatFromCurve(FloatCurve=DataInterface() {propertyValues {Curve {keys [
            {arriveTangent=2.6957, interpMode=RCIM_Cubic, leaveTangent=2.6957, tangentMode=RCTM_User, time=0, value=0.4743}
            {arriveTangent=0.5523, interpMode=RCIM_Cubic, leaveTangent=0.5523, time=0.0843, value=1.0291}
            {arriveTangent=-1.1238, interpMode=RCIM_Cubic, leaveTangent=-1.1238, time=0.9354, value=0.9909}
            {arriveTangent=-1.4703, interpMode=RCIM_Cubic, leaveTangent=-1.4703, tangentMode=RCTM_User, time=1, value=0}
           ]}}})
      )
      B=$System.ExplosionScale
     )
     Vector2D=<1 1>
    )
    Write to Module Namespace=true
   )
  ]
  rendererClasses [
   &NiagaraMeshRendererProperties
   &NiagaraLightRendererProperties
   &NiagaraSpriteRendererProperties
  ]
  renderers [
   {rendererClass=&NiagaraMeshRendererProperties, rendererIndex=0}
   {rendererClass=&NiagaraLightRendererProperties, rendererIndex=1}
   {rendererClass=&NiagaraMeshRendererProperties, rendererIndex=2}
   {rendererClass=&NiagaraMeshRendererProperties, rendererIndex=3}
   {rendererClass=&NiagaraSpriteRendererProperties, rendererIndex=4}
  ]
 }
 NE_Flare001 {
  bEnabled=false
  emitterUpdateScript [
   &EmitterState(Life Cycle Mode=::NewEnumerator1, Loop Behavior=::NewEnumerator1, Loop Delay=$System.ExplosionDelay, Loop Duration=0.4, UseLoopDelay=true)
   &SpawnBurst_Instantaneous(Spawn Count=1, Use Spawn Probability=false)
   &SpawnBurst_Instantaneous(Spawn Count=1, Spawn Group=1, Use Spawn Probability=false) {module=SpawnBurst_Instantaneous001}
  ]
  particleSpawnScript [
   &InitializeParticle(
    Lifetime=0.7
    Mass Mode=::NewEnumerator1
    Mesh Scale=&Multiply_VectorByFloat(Vector=<1 1 1>, Float=$System.ExplosionScale)
    Mesh Scale Mode=::NewEnumerator4
    Position Offset=$System.ExplosionPosition
    Sprite Size Mode=::NewEnumerator3
    Uniform Sprite Size=&Multiply_Float(A=$System.ExplosionScale, B=10)
    UsePositionOffset=true
    Write Material Random=true
   ) {module=InitializeParticle001}
   /SetVariables(Particles.VisibilityTag=&SpawnGroup()) {isSetParameters=true, module=SetVariables_#3}
  ]
  particleUpdateScript [
   &ParticleState()
   &ScaleMeshSize(Scale Factor=&VectorFromFloat(Value=&MaskFloatBySpawnGroup(
      Masked Float=&FloatFromCurve(FloatCurve=DataInterface() {propertyValues {Curve {keys [{arriveTangent=5.2157, interpMode=RCIM_Cubic, leaveTangent=5.2157, tangentMode=RCTM_User, time=0, value=0.3}, <at 0.4 1 RCIM_Cubic>]}}})
      Passthrough Float=&FloatFromCurve(FloatCurve=DataInterface() {propertyValues {Curve {keys [
           {interpMode=RCIM_Cubic, tangentMode=RCTM_Break, time=0, value=0.2}
           {arriveTangent=2.9365, interpMode=RCIM_Cubic, leaveTangent=2.9365, tangentMode=RCTM_User, time=0.1, value=1}
           {interpMode=RCIM_Cubic, tangentMode=RCTM_Break, time=1, value=2}
          ]}}})
     )))
   &ScaleColor(
    Scale Alpha=&MaskFloatBySpawnGroup(
     Masked Float=0
     Passthrough Float=&FloatFromCurve(FloatCurve=DataInterface() {propertyValues {Curve {keys [
          {interpMode=RCIM_Cubic, tangentMode=RCTM_User, time=0, value=1}
          {arriveTangent=-2.6886, interpMode=RCIM_Cubic, leaveTangent=-2.6886, tangentMode=RCTM_User, time=1, value=0}
         ]}}})
    )
    Scale RGB=&MaskVectorBySpawnGroup(
     Masked Vector=&Multiply_Vector(
      A=&MakeVectorFromLinearColorRGB(Color=&ColorFromCurve(ColorCurve=DataInterface() {propertyValues {
          AlphaCurve {keys [<at 0 1>,<at 0.4696 1>,<at 1 0>]}
          BlueCurve {keys [<at 0 5>,<at 0.4696 0.0232>,<at 1 0.037>]}
          GreenCurve {keys [<at 0 3.6578>,<at 0.4696 0.1988>,<at 1 0.0811>]}
          RedCurve {keys [<at 0 0.0913>,<at 0.4696 0.6979>,<at 1 0.3698>]}
         }}))
      B=&VectorFromFloat(Value=&FloatFromCurve(FloatCurve=DataInterface() {propertyValues {Curve {keys [<at 0 0.7>,<at 0.7729 0.7>,<at 1 0>]}}}, Scale Curve=0.2))
     )
     Passthrough Vector=&Multiply_Vector(
      A=&MakeVectorFromLinearColorRGB(Color=&ColorFromCurve(ColorCurve=DataInterface() {propertyValues {
          AlphaCurve {keys [<at 0 1>,<at 0.4696 1>,<at 1 0>]}
          BlueCurve {keys [<at 0 1>,<at 0.2681 1>,<at 0.4696 0.0507>,<at 1 0.037>]}
          GreenCurve {keys [<at 0 0.823>,<at 0.2681 0.3862>,<at 0.4696 0.2082>,<at 1 0.0811>]}
          RedCurve {keys [<at 0 0.0292>,<at 0.2681 0.0006>,<at 0.4696 0.6979>,<at 1 0.3698>]}
         }}))
      B=&VectorFromFloat(Value=&FloatFromCurve(FloatCurve=DataInterface() {propertyValues {Curve {keys [
            {interpMode=RCIM_Cubic, tangentMode=RCTM_Break, time=0, value=10}
            {arriveTangent=-6.6072, interpMode=RCIM_Cubic, leaveTangent=-6.6072, tangentMode=RCTM_User, time=0.15, value=3}
            {interpMode=RCIM_Cubic, tangentMode=RCTM_Break, time=1, value=1}
           ]}}}))
     )
    )
    ScaleA=true
    ScaleRGB=true
   )
   &DynamicMaterialParameters(
    Index 0 Param 1=&FloatFromCurve(FloatCurve=DataInterface() {propertyValues {Curve {keys [{arriveTangent=-2.1015, interpMode=RCIM_Cubic, leaveTangent=-2.1015, tangentMode=RCTM_User, time=0, value=0.5}, <at 1 -0.15>]}}})
    Index 0 Param 2=&FloatFromCurve(FloatCurve=DataInterface() {propertyValues {Curve {keys [{arriveTangent=-0.3464, interpMode=RCIM_Cubic, leaveTangent=-0.3464, tangentMode=RCTM_User, time=0, value=1}, <at 1 0.9>]}}})
    Index 0 Param 3=&FloatFromCurve(FloatCurve=DataInterface() {propertyValues {Curve {keys [
         {interpMode=RCIM_Cubic, tangentMode=RCTM_Break, time=0, value=0}
         {arriveTangent=0.4478, interpMode=RCIM_Cubic, leaveTangent=0.4478, tangentMode=RCTM_User, time=0.15, value=1}
         {interpMode=RCIM_Cubic, tangentMode=RCTM_Break, time=1, value=1.2}
        ]}}})
    Index 3 Param 1=1
    Module.Param0WriteEnabled=true
    Module.Param0WriteEnabled3=true
    Module.Param1WriteEnabled=true
    Module.Param1WriteEnabled3=true
    Module.Param2WriteEnabled=true
    Module.Param2WriteEnabled3=true
    Module.Param3WriteEnabled=false
    Module.Param3WriteEnabled3=true
    Write Parameter Index 3=true
   )
   &FN_Mesh_Attributes(
    Color=&MultiplyLinearColor(Linear Color=<rgba 1 0.4513 0 0.1652>, Scale Factor=<1 1 1 1>)
    Create Color Attribute=true
    Create Scale Attribute=true
    Scale=&VectorFromFloat(Value=&Multiply_Float(
      A=5
      B=&Multiply_Float(
       A=0.2
       B=&FloatFromCurve(FloatCurve=DataInterface() {propertyValues {Curve {
           defaultValue=0.1
           keys [
            {arriveTangent=9.0725, interpMode=RCIM_Cubic, leaveTangent=9.0725, tangentMode=RCTM_User, time=0, value=0}
            {arriveTangentWeight=0.1274, interpMode=RCIM_Cubic, leaveTangentWeight=0.2174, tangentMode=RCTM_User, time=0.1408, value=0.9847}
            {arriveTangentWeight=0.2612, interpMode=RCIM_Cubic, leaveTangentWeight=0.0303, tangentMode=RCTM_User, time=0.4791, value=0.958}
            {arriveTangent=-1.8392, interpMode=RCIM_Cubic, leaveTangent=-1.8392, time=0.594, value=0.2084}
            {arriveTangent=-0.0405, interpMode=RCIM_Cubic, leaveTangent=-0.0405, tangentMode=RCTM_User, time=1, value=0}
           ]
          }}})
      )
     ))
    Write to Module Namespace=true
   )
   /SetVariables(Particles.RefracSphereDynMat:Vector4f=&MakeVector4(
     X=&Lerp_Float(
      A=1
      B=&FloatFromCurve(FloatCurve=DataInterface() {propertyValues {Curve {keys [<at 0 1>,<at 0.3172 1>,<at 1 1>]}}}, Scale Curve=0)
      Alpha=&FN_PickFloatDependOnPlatform(TargetPlatformActive=1, TargetPlatformNoActive=0)
     )
     Y=&FloatFromCurve(FloatCurve=DataInterface() {propertyValues {Curve {keys [<at 0 15>,<at 1 0>]}}})
     Z=0
     W=0
    )) {isSetParameters=true, module=SetVariables_#4}
   &FN_Sprite_Attributes(
    Camera Offset=20
    Color=&MultiplyLinearColor(
     Linear Color=$Particles.Color
     Scale Factor=&MakeVector4FromVectorAndFloat(
      "Vector (RGB)"=&VectorFromFloat(Value=5)
      "Float (Alpha)"=&Waveform("Waveform [1]"=::NewEnumerator0, "[1] Frequency"=10, "[1] Amplitude Min/Max"=<0.2 1>)
     )
    )
    Create Camera Offset Attribute=true
    Create Color Attribute=true
    Create SpriteSize Attribute=true
    SpriteSize=&Multiply_Vector2D_ByFloat(
     Float=&Multiply_Float(
      A=&Multiply_Float(
       A=500
       B=&FloatFromCurve(FloatCurve=DataInterface() {propertyValues {Curve {
           defaultValue=0
           keys [
            {arriveTangent=9.0725, interpMode=RCIM_Cubic, leaveTangent=9.0725, tangentMode=RCTM_User, time=0, value=0}
            {arriveTangent=-2.5249, interpMode=RCIM_Cubic, leaveTangent=-2.5249, tangentMode=RCTM_User, time=0.1408, value=0.9847}
            {arriveTangent=-0.0405, interpMode=RCIM_Cubic, leaveTangent=-0.0405, tangentMode=RCTM_User, time=1, value=0}
           ]
          }}})
      )
      B=0.1
     )
     Vector2D=<1 1>
    )
    Write to Module Namespace=true
   )
   &FN_Light_Attributes(
    Color=<rgba 10 3.6259 0 1>
    Create Color Attribute=true
    Create Exponent Attribute=true
    Create Position Attribute=true
    Create Radius Attribute=true
    Exponent=1.5
    Position=&AddVectorToPosition(Position=/Niagara/DynamicInputs/Helpers/SimulationPosition(), Vector=$System.ExplosionPosition)
    Radius=&Multiply_Float(A=130, B=&FN_FrontEndFloatModifier(Gameplay=0.952, Front End=1))
    Scale Color=&FloatFromCurve(
     FloatCurve=DataInterface() {propertyValues {Curve {keys [
         {interpMode=RCIM_Cubic, tangentMode=RCTM_User, time=0, value=1}
         {arriveTangent=-0.5426, interpMode=RCIM_Cubic, leaveTangent=-0.5426, tangentMode=RCTM_User, time=1, value=0}
        ]}}}
     Scale Curve=10
    )
   )
  ]
  rendererClasses [
   &NiagaraMeshRendererProperties
   &NiagaraSpriteRendererProperties
   &NiagaraLightRendererProperties
  ]
  renderers [
   {rendererClass=&NiagaraMeshRendererProperties, rendererIndex=0}
   {rendererClass=&NiagaraSpriteRendererProperties, rendererIndex=1}
   {rendererClass=&NiagaraLightRendererProperties, rendererIndex=2}
  ]
 }
 SmokeLWE {rendererClasses [], renderers []}
 Smoke_Sphere {
  emitterUpdateScript [
   &EmitterState(Life Cycle Mode=::NewEnumerator1, Loop Behavior=::NewEnumerator1, Loop Delay=$System.ExplosionDelay, Loop Duration=0.5, UseLoopDelay=true)
   &SpawnBurst_Instantaneous(Spawn Count=15, Spawn Time=0.05, Use Spawn Probability=false)
   &SpawnBurst_Instantaneous(Spawn Count=9, Spawn Group=1, Use Spawn Probability=false, UseLoopCountLimit=false) {enabled=false, module=SpawnBurst_Instantaneous001}
  ]
  particleSpawnScript [
   &InitializeParticle(
    Color=<rgba 0.2502 0.2502 0.2502 1>
    Lifetime=0.35
    Mass Mode=::NewEnumerator1
    Position Offset=$System.ExplosionPosition
    Sprite Rotation Mode=::NewEnumerator2
    Sprite Size=&Multiply_Vector2D(
     A=&Vector2DFromFloat(Value=&RandomRangeFloat(Minimum=100, Maximum=200))
     B=&Vector2DFromFloat(Value=$System.ExplosionScale)
    )
    Sprite Size Mode=::NewEnumerator4
    UsePositionOffset=true
    Write Material Random=true
   ) {module=InitializeParticle002}
   /SetVariables(Particles.InitSpriteRotRate=&UniformRangedFloat(Minimum=-3, Maximum=3)) {isSetParameters=true}
   &SphereLocation(
    MaskSpawnGroup=false
    Sphere Distribution:ENiagaraSphereDistributionMode=::NewEnumerator1
    Sphere Radius=&Multiply_Float(A=205, B=$System.ExplosionScale)
    U Position:NiagaraFloat=&ReturnNormalizedExecIndex(Normalized Index Includes One=false)
    UseNonUniformScale=false
    UseOffset=false
    UseSphereOrientation=false
    V Position:NiagaraFloat=&MaskFloatBySpawnGroup(
     Masked Float=&RandomRangeFloat(Minimum=0.9, Maximum=1)
     Passthrough Float=&RandomRangeFloat(Minimum=0.5, Maximum=0.56)
    )
   ) {enabled=false}
   /SetVariables(Particles.Position=&Add_Vector(A=$System.ExplosionPosition, B=<0 0 0>)) {enabled=false, isSetParameters=true}
   &AddVelocity(
    Velocity=$Output.SphereLocation.SphereNormal
    Velocity Speed Scale=&Multiply_Float(
     A=&MaskFloatBySpawnGroup(
      Masked Float=&RandomRangeFloat(Minimum=450, Maximum=2000)
      Passthrough Float=&RandomRangeFloat(Minimum=2000, Maximum=2700)
     )
     B=$System.ExplosionScale
    )
   ) {enabled=false}
  ]
  particleUpdateScript [
   &ParticleState()
   &SpriteRotationRate(Rotation Rate=&Multiply_Float(
     A=$Particles.InitSpriteRotRate
     B=&FloatFromCurve(
      FloatCurve=DataInterface() {propertyValues {Curve {keys [
          {arriveTangent=-4.0107, interpMode=RCIM_Cubic, leaveTangent=-4.0107, tangentMode=RCTM_User, time=-0.0046, value=1.0113}
          {arriveTangent=-0.2496, interpMode=RCIM_Cubic, leaveTangent=-0.2496, tangentMode=RCTM_User, time=0.5, value=0.16}
         ]}}}
      CurveIndex=$Emitter.Age
      Scale Curve=150
     )
    ))
   &ScaleColor(
    Scale Alpha=&FloatFromCurve(
     FloatCurve=DataInterface() {propertyValues {Curve {keys [
         {arriveTangent=-6.7619, interpMode=RCIM_Cubic, leaveTangent=-6.7619, tangentMode=RCTM_User, time=0, value=0}
         {interpMode=RCIM_Cubic, tangentMode=RCTM_User, time=0.2, value=0.9}
         {arriveTangent=-1.125, interpMode=RCIM_Cubic, leaveTangent=-1.125, time=0.6486, value=0.8794}
         {interpMode=RCIM_Cubic, tangentMode=RCTM_User, time=1, value=0}
        ]}}}
     Scale Curve=0.7
    )
    ScaleA=true
    ScaleRGB=false
   )
   &ScaleColor(
    Scale Alpha=&FloatFromCurve(
     FloatCurve=DataInterface() {propertyValues {Curve {keys [
         {arriveTangent=-6.7619, interpMode=RCIM_Cubic, leaveTangent=-6.7619, tangentMode=RCTM_User, time=0, value=3}
         {interpMode=RCIM_Cubic, tangentMode=RCTM_User, time=0.2, value=1.5}
         {interpMode=RCIM_Cubic, tangentMode=RCTM_User, time=1, value=0}
        ]}}}
     Scale Curve=0.7
    )
    ScaleA=true
    ScaleRGB=true
   ) {enabled=false, module=ScaleColor001}
   /Niagara/Modules/Update/Forces/CurlNoiseForce(Noise Frequency=0.0012, Noise Strength=680, Pan Noise=true, Pan Noise Field=<0 0 1>) {enabled=false}
   &Drag(
    Drag=&FloatFromCurve(
     FloatCurve=DataInterface() {propertyValues {Curve {keys [
         {interpMode=RCIM_Cubic, leaveTangent=0.0927, tangentMode=RCTM_Break, time=0, value=1}
         {arriveTangent=-2.586, interpMode=RCIM_Cubic, leaveTangent=-2.586, tangentMode=RCTM_User, time=1, value=0}
        ]}}}
     Scale Curve=&Lerp_Float(
      A=5
      B=&RandomRangeFloat(Minimum=7, Maximum=8)
      Alpha=&MakeFloatFromInt(Int=&SpawnGroup())
     )
    )
    Use Linear Drag=true
    Use Rotational Drag=false
   )
   &SolveForcesAndVelocity(Clamp Velocity=false, Limit Acceleration=false)
   &ScaleSpriteSize(Scale Factor=&Vector2DFromFloat(Value=&FloatFromCurve(FloatCurve=DataInterface() {propertyValues {Curve {keys [
          {arriveTangent=0.55, interpMode=RCIM_Cubic, leaveTangent=0.55, tangentMode=RCTM_User, time=0, value=1}
          {arriveTangent=-1, interpMode=RCIM_Cubic, leaveTangent=-1, time=0.9424, value=0.8548}
          {arriveTangent=-3.194, interpMode=RCIM_Cubic, leaveTangent=-3.194, tangentMode=RCTM_User, time=1, value=0}
         ]}}})))
   &DynamicMaterialParameters(
    Index 0 Param 1=&Add_Float(
     A=&RandomRangeFloat(Minimum=0, Maximum=1)
     B=&FloatFromCurve(
      FloatCurve=DataInterface() {propertyValues {Curve {keys [
          {arriveTangent=2.7533, interpMode=RCIM_Cubic, leaveTangent=1.8958, tangentMode=RCTM_User, time=0, value=1}
          {arriveTangent=0.8292, interpMode=RCIM_Cubic, leaveTangent=0.8292, tangentMode=RCTM_User, time=0.321, value=1.3453}
          {arriveTangent=0.0529, interpMode=RCIM_Cubic, leaveTangent=0.2021, tangentMode=RCTM_User, time=1, value=1.5}
         ]}}}
      CurveIndex=&Add_Float(A=&Multiply_Float(A=$Particles.Initial.MaterialRandom, B=0.5), B=$Particles.NormalizedAge)
     )
    )
    Index 0 Param 2=&FloatFromCurve(FloatCurve=DataInterface() {propertyValues {Curve {keys [
         {arriveTangent=0.0453, interpMode=RCIM_Cubic, tangentMode=RCTM_Break, time=0, value=-1}
         {arriveTangent=0.0604, interpMode=RCIM_Cubic, leaveTangent=0.0604, tangentMode=RCTM_User, time=1, value=-1}
        ]}}})
    Index 0 Param 3=&FloatFromCurve(
     FloatCurve=DataInterface() {propertyValues {Curve {keys [
         {arriveTangent=0.9822, interpMode=RCIM_Cubic, leaveTangent=0.9822, tangentMode=RCTM_User, time=0, value=1}
         {arriveTangent=0.2347, interpMode=RCIM_Cubic, leaveTangent=-0.7437, tangentMode=RCTM_Break, time=0.25, value=0}
        ]}}}
     Scale Curve=0
    )
    Index 0 Param 4=&FloatFromCurve(FloatCurve=DataInterface() {propertyValues {Curve {keys [<at 0.5 1>,<at 1 0>]}}})
    Index 1 Param 1=&FloatFromCurve(
     FloatCurve=DataInterface() {propertyValues {Curve {keys [
         {interpMode=RCIM_Cubic, leaveTangent=4.7088, tangentMode=RCTM_Break, time=0, value=0}
         {arriveTangent=4.2257, interpMode=RCIM_Cubic, leaveTangent=4.2257, tangentMode=RCTM_User, time=0.25, value=1}
        ]}}}
     Scale Curve=&RandomRangeFloat(Minimum=1.5, Maximum=1.5)
    )
    Index 1 Param 2=&FloatFromCurve(
     FloatCurve=DataInterface() {propertyValues {Curve {keys [<at 0 0.2>, {leaveTangent=0.9066, time=1, value=2}]}}}
     Scale Curve=&RandomRangeFloat(Minimum=0.5, Maximum=0.6)
    )
    Index 1 Param 3=&FloatFromCurve(FloatCurve=DataInterface() {propertyValues {Curve {keys [
         {interpMode=RCIM_Cubic, leaveTangent=6.3917, tangentMode=RCTM_Break, time=0, value=0.2}
         {arriveTangent=1.7549, interpMode=RCIM_Cubic, leaveTangent=10.6767, tangentMode=RCTM_User, time=0.25, value=1}
        ]}}})
    Index 1 Param 4=&FloatFromCurve(
     FloatCurve=DataInterface() {propertyValues {Curve {keys [{arriveTangent=-0.136, interpMode=RCIM_Cubic, leaveTangent=-0.136, tangentMode=RCTM_User, time=0, value=1}, <at 1 0.95>]}}}
     Scale Curve=&RandomRangeFloat(Minimum=1.1, Maximum=1.4)
    )
    Index 2 Param 1=&FloatFromCurve(
     FloatCurve=DataInterface() {propertyValues {Curve {keys [
         {interpMode=RCIM_Cubic, leaveTangent=1.825, tangentMode=RCTM_Break, time=0, value=0}
         {arriveTangent=-0.2149, tangentMode=RCTM_User, time=1, value=1}
        ]}}}
     Scale Curve=2
    )
    Index 2 Param 2=&FloatFromCurve(
     FloatCurve=DataInterface() {propertyValues {Curve {keys [
         {arriveTangent=1.9056, interpMode=RCIM_Cubic, leaveTangent=1.9056, tangentMode=RCTM_User, time=0, value=0.2}
         {arriveTangent=0.0259, interpMode=RCIM_Cubic, leaveTangent=0.0259, tangentMode=RCTM_User, time=1, value=1.4}
        ]}}}
     Scale Curve=&RandomRangeFloat(Minimum=0.2, Maximum=0.7)
    )
    Index 2 Param 3=&FloatFromCurve(FloatCurve=DataInterface() {propertyValues {Curve {keys [
         {interpMode=RCIM_Cubic, leaveTangent=2.3064, tangentMode=RCTM_Break, time=0, value=1}
         {interpMode=RCIM_Cubic, leaveTangent=11.1233, tangentMode=RCTM_Break, time=1, value=2}
        ]}}})
    Index 2 Param 4=&FloatFromCurve(
     FloatCurve=DataInterface() {propertyValues {Curve {keys [
         {arriveTangent=-0.6867, interpMode=RCIM_Cubic, leaveTangent=-0.6867, tangentMode=RCTM_User, time=0, value=1}
         {arriveTangent=-0.2589, interpMode=RCIM_Cubic, leaveTangent=-0.2589, tangentMode=RCTM_User, time=0.3, value=0.8}
         {interpMode=RCIM_Cubic, leaveTangent=-1.2956, tangentMode=RCTM_Break, time=1, value=0.75}
        ]}}}
     Scale Curve=1.6
    )
    Module.Index0Float0Write=true
    Module.Index0Float1Write=true
    Module.Index0Float2Write=true
    Module.Index0Float3Write=true
    Module.Index1Float0Write=true
    Module.Index1Float1Write=true
    Module.Index1Float2Write=true
    Module.Index1Float3Write=true
    Module.Index2Float0Write=true
    Module.Index2Float1Write=true
    Module.Index2Float2Write=true
    Module.Index2Float3Write=true
    Write Parameter Index 1=::NewEnumerator1
    Write Parameter Index 2=::NewEnumerator1
   )
   &DynamicMaterialParameters(
    Index 0 Param 1=&Add_Float(
     A=$Particles.RandomFloat
     B=&FloatFromCurve(
      FloatCurve=DataInterface() {propertyValues {Curve {keys [
          {arriveTangent=2.7533, interpMode=RCIM_Cubic, leaveTangent=1.8958, tangentMode=RCTM_User, time=0, value=1}
          {arriveTangent=0.8292, interpMode=RCIM_Cubic, leaveTangent=0.8292, tangentMode=RCTM_User, time=0.321, value=1.3453}
          {arriveTangent=0.0529, interpMode=RCIM_Cubic, leaveTangent=0.2021, tangentMode=RCTM_User, time=1, value=1.5}
         ]}}}
      CurveIndex=&Add_Float(A=&Multiply_Float(A=$Particles.Initial.MaterialRandom, B=0.5), B=$Particles.NormalizedAge)
     )
    )
    Index 0 Param 2=&FloatFromCurve(FloatCurve=DataInterface() {propertyValues {Curve {keys [
         {arriveTangent=0.0453, interpMode=RCIM_Cubic, tangentMode=RCTM_Break, time=0, value=-1}
         {arriveTangent=0.0604, interpMode=RCIM_Cubic, leaveTangent=0.0604, tangentMode=RCTM_User, time=1, value=-1}
        ]}}})
    Index 0 Param 3=&FloatFromCurve(
     FloatCurve=DataInterface() {propertyValues {Curve {keys [
         {arriveTangent=0.9822, interpMode=RCIM_Cubic, leaveTangent=0.9822, tangentMode=RCTM_User, time=0, value=1}
         {arriveTangent=0.2347, interpMode=RCIM_Cubic, leaveTangent=-0.7437, tangentMode=RCTM_Break, time=0.25, value=0}
        ]}}}
     Scale Curve=&Lerp_Float(A=0.1, B=1, Alpha=$System.CameraFadeAlpha)
    )
    Index 0 Param 4=&FloatFromCurve(FloatCurve=DataInterface() {propertyValues {Curve {keys [<at 0.5 1>,<at 1 0>]}}})
    Index 1 Param 1=&FloatFromCurve(
     FloatCurve=DataInterface() {propertyValues {Curve {keys [
         {interpMode=RCIM_Cubic, leaveTangent=4.7088, tangentMode=RCTM_Break, time=0, value=0}
         {arriveTangent=4.2257, interpMode=RCIM_Cubic, leaveTangent=4.2257, tangentMode=RCTM_User, time=0.25, value=1}
        ]}}}
     Scale Curve=&RandomRangeFloat(Minimum=1.5, Maximum=1.5)
    )
    Index 1 Param 2=&FloatFromCurve(
     FloatCurve=DataInterface() {propertyValues {Curve {keys [<at 0 0.2>, {leaveTangent=0.9066, time=1, value=2}]}}}
     Scale Curve=&RandomRangeFloat(Minimum=0.5, Maximum=0.6)
    )
    Index 1 Param 3=&FloatFromCurve(FloatCurve=DataInterface() {propertyValues {Curve {keys [
         {interpMode=RCIM_Cubic, leaveTangent=6.3917, tangentMode=RCTM_Break, time=0, value=0.2}
         {arriveTangent=1.7549, interpMode=RCIM_Cubic, leaveTangent=10.6767, tangentMode=RCTM_User, time=0.25, value=1}
        ]}}})
    Index 1 Param 4=&FloatFromCurve(
     FloatCurve=DataInterface() {propertyValues {Curve {keys [{arriveTangent=-0.136, interpMode=RCIM_Cubic, leaveTangent=-0.136, tangentMode=RCTM_User, time=0, value=1}, <at 1 0.95>]}}}
     Scale Curve=&RandomRangeFloat(Minimum=1.1, Maximum=1.4)
    )
    Index 2 Param 1=&FloatFromCurve(
     FloatCurve=DataInterface() {propertyValues {Curve {keys [
         {interpMode=RCIM_Cubic, leaveTangent=1.825, tangentMode=RCTM_Break, time=0, value=0}
         {arriveTangent=-0.2149, tangentMode=RCTM_User, time=1, value=1}
        ]}}}
     Scale Curve=2
    )
    Index 2 Param 2=&FloatFromCurve(
     FloatCurve=DataInterface() {propertyValues {Curve {keys [
         {arriveTangent=1.9056, interpMode=RCIM_Cubic, leaveTangent=1.9056, tangentMode=RCTM_User, time=0, value=0.2}
         {arriveTangent=0.0259, interpMode=RCIM_Cubic, leaveTangent=0.0259, tangentMode=RCTM_User, time=1, value=1.4}
        ]}}}
     Scale Curve=&RandomRangeFloat(Minimum=0.2, Maximum=0.7)
    )
    Index 2 Param 3=&FloatFromCurve(FloatCurve=DataInterface() {propertyValues {Curve {keys [
         {interpMode=RCIM_Cubic, leaveTangent=2.3064, tangentMode=RCTM_Break, time=0, value=1}
         {interpMode=RCIM_Cubic, leaveTangent=11.1233, tangentMode=RCTM_Break, time=1, value=2}
        ]}}})
    Index 2 Param 4=&FloatFromCurve(
     FloatCurve=DataInterface() {propertyValues {Curve {keys [
         {arriveTangent=-0.6867, interpMode=RCIM_Cubic, leaveTangent=-0.6867, tangentMode=RCTM_User, time=0, value=1}
         {arriveTangent=-0.2589, interpMode=RCIM_Cubic, leaveTangent=-0.2589, tangentMode=RCTM_User, time=0.3, value=0.8}
         {interpMode=RCIM_Cubic, leaveTangent=-1.2956, tangentMode=RCTM_Break, time=1, value=0.75}
        ]}}}
     Scale Curve=1.6
    )
    Module.Param0WriteEnabled=true
    Module.Param0WriteEnabled1=true
    Module.Param0WriteEnabled2=true
    Module.Param1WriteEnabled=false
    Module.Param1WriteEnabled1=true
    Module.Param1WriteEnabled2=true
    Module.Param2WriteEnabled=true
    Module.Param2WriteEnabled1=true
    Module.Param2WriteEnabled2=true
    Module.Param3WriteEnabled=true
    Module.Param3WriteEnabled1=true
    Module.Param3WriteEnabled2=true
    Write Parameter Index 1=true
    Write Parameter Index 2=true
   ) {enabled=false, module=DynamicMaterialParameters001}
   &CameraOffset()
   &RotateAroundPoint(
    Radius:NiagaraFloat=&Multiply_Float(
     A=&Lerp_Float(
      A=0
      B=&RandomRangeFloat(Minimum=50, Maximum=30)
      Alpha=&FloatFromCurve(FloatCurve=DataInterface() {propertyValues {Curve {keys [
           {arriveTangent=1.5757, interpMode=RCIM_Cubic, leaveTangent=1.5757, tangentMode=RCTM_User, time=0, value=1}
           {arriveTangent=-1, interpMode=RCIM_Cubic, leaveTangent=-1, time=0.9352, value=1.0411}
           {arriveTangent=-1.9265, interpMode=RCIM_Cubic, leaveTangent=-1.9265, tangentMode=RCTM_User, time=1, value=0}
          ]}}})
     )
     B=$System.ExplosionScale
    )
    Rotation Center:Vector3f=$System.ExplosionPosition
    Rotation Phase:NiagaraFloat=&ModuloFloat(A=&ReturnNormalizedExecIndex(), B=1)
    X Axis:Vector3f=&RandomVector(Vector Scale=1)
   )
  ]
  rendererClasses [&NiagaraSpriteRendererProperties]
  renderers [
   {rendererClass=&NiagaraSpriteRendererProperties}
   {rendererClass=&NiagaraSpriteRendererProperties, rendererIndex=1}
  ]
 }
 Streaks_Fat {
  emitterUpdateScript [
   &EmitterState(Life Cycle Mode=::NewEnumerator1, Loop Behavior=::NewEnumerator1, Loop Delay=$System.ExplosionDelay, Loop Duration=0.4, UseLoopDelay=true)
   &SpawnBurst_Instantaneous(Spawn Count=18, Use Spawn Probability=false)
  ]
  particleSpawnScript [
   &InitializeParticle(
    Color=<rgba 0.0036 0.0481 2 1>
    Lifetime=&UniformRangedFloat(Minimum=0.15, Maximum=0.25)
    Mass Mode=::NewEnumerator1
    Position Offset=$System.ExplosionPosition
    Sprite Size=&Multiply_Vector2D_ByFloat(Float=$System.ExplosionScale, Vector2D=&UniformRangedVector2D(Minimum=<10 60>, Maximum=<20 80>))
    Sprite Size Mode=::NewEnumerator4
    UsePositionOffset=true
    Write Material Random=true
   ) {module=InitializeParticle001}
   /Niagara/Modules/Spawn/Location/SphereLocation(
    MaskSpawnGroup=false
    Module.Override Randomness=false
    Sphere Radius=&Multiply_Float(A=5, B=&FN_FrontEndFloatModifier(Gameplay=1, Front End=$System.FrontEndScale))
    Surface Only=true
   )
   &AddVelocityFromPoint(
    Module.Use Falloff Distance=false
    Velocity Origin:Vector3f=$System.ExplosionPosition
    Velocity Strength=&Multiply_Float(
     A=&UniformRangedFloat(Minimum=300, Maximum=400)
     B=&FN_FrontEndFloatModifier(Gameplay=1, Front End=$System.FrontEndScale)
    )
   )
  ]
  particleUpdateScript [
   &ParticleState()
   &SolveForcesAndVelocity(Clamp Velocity=false, Limit Acceleration=false)
   &ScaleSpriteSize(Scale Factor=&Multiply_Vector2D_ByFloat(
     Float=&FloatFromCurve(FloatCurve=DataInterface() {propertyValues {Curve {
         defaultValue=1
         keys [
          {arriveTangent=9.0725, interpMode=RCIM_Cubic, leaveTangent=9.0725, tangentMode=RCTM_User, time=0, value=0}
          {arriveTangent=-2.5249, interpMode=RCIM_Cubic, leaveTangent=-2.5249, tangentMode=RCTM_User, time=0.1408, value=0.9847}
          {arriveTangent=-0.0405, interpMode=RCIM_Cubic, leaveTangent=-0.0405, tangentMode=RCTM_User, time=1, value=0}
         ]
        }}})
     Vector2D=&Vector2DFromCurve(
      Vector2Curve=DataInterface() {propertyValues {XCurve {keys [<at 0 0.75 RCIM_Cubic>,<at 1 1 RCIM_Cubic>]}, YCurve {keys [<at 0 1 RCIM_Cubic>,<at 1 2 RCIM_Cubic>]}}}
      ScaleCurve=<1 1>
     )
    ))
   &ScaleColor(
    Scale Alpha=&FloatFromCurve(FloatCurve=DataInterface() {propertyValues {Curve {keys [<at 0 1 RCIM_Cubic>,<at 1 0>]}}})
    Scale RGB=&VectorFromFloat(Value=&FloatFromCurve(FloatCurve=DataInterface() {propertyValues {Curve {keys [<at 0 10 RCIM_Cubic>,<at 1 1 RCIM_Cubic>]}}}))
    ScaleA=true
    ScaleRGB=true
   )
  ]
  rendererClasses [&NiagaraSpriteRendererProperties]
  renderers [{rendererClass=&NiagaraSpriteRendererProperties}]
 }
 Wave02001 {
  bEnabled=false
  emitterUpdateScript [
   &EmitterState(Life Cycle Mode=::NewEnumerator1, Loop Behavior=::NewEnumerator1, Loop Delay=$System.ExplosionDelay, Loop Duration=1, UseLoopDelay=true)
   &SpawnBurst_Instantaneous(Use Spawn Probability=false, UseLoopCountLimit=false)
   &SpawnBurst_Instantaneous(Spawn Group=1, Spawn Time=0.05, Use Spawn Probability=false, UseLoopCountLimit=false) {module=SpawnBurst_Instantaneous001}
  ]
  particleSpawnScript [
   &InitializeParticle(
    Color Mode=::NewEnumerator0
    Lifetime=&MaskFloat(Mask Source=0, Mask Parameter=?, Mask=0, Masked Float=1.2, Passthrough Float=0.25)
    Mesh Scale Mode=::NewEnumerator3
    Mesh Uniform Scale=&Multiply_Float(
     A=$System.ExplosionScale
     B=&MaskFloat(Mask Source=0, Mask Parameter=?, Mask=0, Masked Float=0.8, Passthrough Float=1)
    )
    Position Offset=$System.ExplosionPosition
    UsePositionOffset=true
    Write Material Random=true
   )
   &InitialMeshOrientation(Orientation Axis=<0 0 1>, Orientation Vector=<0 0 1>, Rotation=<0 0.25 0>, Use Orientation Vector=true, Use Rotation Vector=true)
   &InitialMeshOrientation(
    Orientation Axis=<0 0 0>
    Orientation Vector=<0 0 0>
    Rotation=&RandomRangeVector(Minimum=<0 0 0>, Maximum=<1 0 0>)
    Use Orientation Vector=false
    Use Rotation Vector=true
   ) {module=InitialMeshOrientation001}
   &CameraOffset(Camera Offset Amount=20)
   /SetVariables(Particles.VisibilityTag=&SpawnGroup()) {isSetParameters=true}
  ]
  particleUpdateScript [
   &ParticleState()
   &ScaleColor(
    Scale Mode=::NewEnumerator1
    Scale RGBA=&MakeVector4_FromLinearColor(LinearColor=&MergeHelper_ColorSwitch(
      0=&ColorFromCurve(ColorCurve=DataInterface() {propertyValues {
         AlphaCurve {keys [<at 0 1>,<at 0.3576 1>,<at 1 0>]}
         BlueCurve {keys [<at 0 1>,<at 0.1084 1>,<at 0.25 1>,<at 1 1>]}
         GreenCurve {keys [<at 0 0>,<at 0.1084 0.5489>,<at 0.25 0.1777>,<at 1 0.0414>]}
         RedCurve {keys [<at 0 0.1667>,<at 0.1084 0.0978>,<at 0.25 0.1029>,<at 1 0.2012>]}
        }})
      1=&ColorFromCurve(ColorCurve=DataInterface() {propertyValues {
         AlphaCurve {keys [<at 0 1>,<at 0.3576 1>,<at 1 0>]}
         BlueCurve {keys [<at 0 1>,<at 0.3 1>,<at 0.8 1>]}
         GreenCurve {keys [<at 0 0.5489>,<at 0.3 0.3272>,<at 0.8 0.0414>]}
         RedCurve {keys [<at 0 0.0978>,<at 0.3 0.1029>,<at 0.8 0.2012>]}
        }})
      2=?
      3=?
      Index=&SpawnGroup()
      Num=::NewEnumerator0
     ))
    ScaleRGBA=true
   )
   &DynamicMaterialParameters(
    Index 0 Param 1=&MergeHelper_FloatSwitch(
     0=&FloatFromCurve(FloatCurve=DataInterface() {propertyValues {Curve {keys [{arriveTangent=-4.0141, interpMode=RCIM_Cubic, leaveTangent=-4.0141, tangentMode=RCTM_User, time=0, value=0.6}, <at 1 -0.7>]}}})
     1=&FloatFromCurve(FloatCurve=DataInterface() {propertyValues {Curve {keys [{arriveTangent=3.726, interpMode=RCIM_Cubic, leaveTangent=3.726, tangentMode=RCTM_User, time=0, value=-0.3}, <at 1 1>]}}})
     2=?
     3=?
     Index=&SpawnGroup()
     Num=::NewEnumerator0
    )
    Index 0 Param 2=&MergeHelper_FloatSwitch(
     0=&FloatFromCurve(FloatCurve=DataInterface() {propertyValues {Curve {keys [{arriveTangent=-0.5818, interpMode=RCIM_Cubic, leaveTangent=-0.5818, tangentMode=RCTM_User, time=0, value=1}, <at 1 0.9>]}}})
     1=&FloatFromCurve(FloatCurve=DataInterface() {propertyValues {Curve {keys [{arriveTangent=-0.5818, interpMode=RCIM_Cubic, leaveTangent=-0.5818, tangentMode=RCTM_User, time=0, value=1}, <at 1 0.8>]}}})
     2=?
     3=?
     Index=&SpawnGroup()
     Num=::NewEnumerator0
    )
    Index 0 Param 3=&FloatFromCurve(FloatCurve=DataInterface() {propertyValues {Curve {keys [<at 0 0>,<at 0.15 1>,<at 1 1.2>]}}})
    Module.Param0WriteEnabled=true
    Module.Param1WriteEnabled=true
    Module.Param2WriteEnabled=true
    Module.Param3WriteEnabled=false
   )
   &ScaleColor(
    Scale Alpha=&FNTimeOfDayScalar(Night Scale=0.7, Morning Scale=0.5, Evening Scale=0.5)
    Scale RGB=&VectorFromFloat(Value=&FloatFromCurve(FloatCurve=DataInterface() {propertyValues {Curve {keys [<at 0 10>,<at 0.15 3>,<at 1 1>]}}}))
    ScaleA=true
    ScaleRGB=true
   ) {module=ScaleColor001}
   &ScaleMeshSize(Scale Factor=&VectorFromFloat(Value=&MergeHelper_FloatSwitch(
      0=&FloatFromCurve(FloatCurve=DataInterface() {propertyValues {Curve {keys [<at 0 0>, {arriveTangent=-6.1514, interpMode=RCIM_Cubic, leaveTangent=-6.1514, tangentMode=RCTM_User, time=0.1, value=2}, <at 1 0>]}}})
      1=&FloatFromCurve(FloatCurve=DataInterface() {propertyValues {Curve {keys [<at 0 0.5>,<at 0.15 1>,<at 1 1.2>]}}})
      2=?
      3=?
      Index=&SpawnGroup()
      Num=::NewEnumerator0
     )))
  ]
  rendererClasses [&NiagaraMeshRendererProperties]
  renderers [
   {rendererClass=&NiagaraMeshRendererProperties, rendererIndex=0}
   {rendererClass=&NiagaraMeshRendererProperties, rendererIndex=1}
  ]
 }
}
summary {
 systemName=NS_Emote_PintPiano_Teleport
 emitters [
  {emitterName=Streaks_Fat, rendererClasses [&NiagaraSpriteRendererProperties]}
  {bEnabled=false, emitterName=LWE_Refract, rendererClasses []}
  {bEnabled=false, emitterName=Main_Smoke001, rendererClasses [&NiagaraMeshRendererProperties]}
  {
   bEnabled=false
   emitterName=NE_Flare001
   rendererClasses [
    &NiagaraMeshRendererProperties
    &NiagaraSpriteRendererProperties
    &NiagaraLightRendererProperties
   ]
  }
  {bEnabled=false, emitterName=Wave02001, rendererClasses [&NiagaraMeshRendererProperties]}
  {
   emitterName=FireTrails
   rendererClasses [
    &NiagaraSpriteRendererProperties
    &NiagaraRibbonRendererProperties
    &NiagaraLightRendererProperties
   ]
  }
  {
   emitterName=NE_Flare
   rendererClasses [
    &NiagaraMeshRendererProperties
    &NiagaraLightRendererProperties
    &NiagaraSpriteRendererProperties
   ]
  }
  {emitterName=Smoke_Sphere, rendererClasses [&NiagaraSpriteRendererProperties]}
  {emitterName=SmokeLWE, rendererClasses []}
 ]
 userVariables [
  User.bIsFrontEnd {type=NiagaraBool}
  User.bIsFrontEndPreview {type=NiagaraBool}
  User.IsCinematic {type=NiagaraBool}
 ]
}
